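# Animation Blueprint for the F_MED_LumpyMuddy female cosmetic body, adding secondary motion to its skirt bones.
# Chained Anim Dynamics nodes start from cached pose 'start', simulating front, side and back bones with angular springs and limits, root-relative or relative to thigh_r.
# A second chain from the same pose repeats the bones with uniform springs and mostly lower angular sim-space influence.
# Unconnected base_waist offset nodes include one weighted by a 'carried' value.
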AnimBlueprint /BRCosmetics/Characters/Player/Female/Medium/Bodies/F_MED_LumpyMuddy/Meshes/F_MED_LumpyMuddy_AnimBP
target_skeleton=&/F_MED_LumpyMuddy_Skeleton
graphs [
 AnimGraph {nodes [
   Anim Dynamics - Bone: dyn_skirt_sd_r_02!(ComponentPose, Alpha=1, GravityOverride:Vector="0,0,-980", ExternalForce="0,0,0") -> (Pose) {
    angularDampingOverride=1
    angularSpringConstant=5000
    bAngularSpring=true
    bOverrideAngularDamping=true
    bUseGravityOverride=true
    gravityScale=0.35
    simulationSpace=BoneRelative
    alphaScaleBiasClamp {bInterpResult=true, interpSpeedDecreasing=1, interpSpeedIncreasing=5}
    boundBone {boneName=dyn_skirt_sd_r_02}
    componentLinearAccScale=<0.3 0.3 0.3>
    componentLinearVelScale=<0.3 0.3 0.3>
    gravityOverride {z=-980}
    physicsBodyDefinitions [{
      boundBone {boneName=dyn_skirt_sd_r_02}
      boxExtents=<20 30 30>
      constraintSetup {angularLimitsMax=<0 0 5>, angularLimitsMin=<0 -50 -5>, angularTarget=<1 0 0>}
      localJointOffset=<4 0 0>
     }]
    relativeSpaceBone {boneName=thigh_r}
    simSpaceSettings {simSpaceAngularAlpha=0.8}
   }
   "Transform (Modify) Bone - Bone: base_waist"(Alpha=1, Translation="0,0,20", Rotation="0,0,0", Scale="1,1,1") {translationMode=BMM_Additive, translationSpace=BCS_BoneSpace, boneToModify {boneName=base_waist}, translation {z=20}}
   "Transform (Modify) Bone - Bone: base_waist"(Alpha=Get carried() -> (carried:double), Translation="0,0,20", Rotation="0,0,0", Scale="1,1,1") {translationMode=BMM_Additive, translationSpace=BCS_BoneSpace, boneToModify {boneName=base_waist}, translation {z=20}}
   n1 = Anim Dynamics - Bone: dyn_skirt_sd_r_03!(ComponentPose=Local To Component(LocalPose=Use cached pose 'start'()), Alpha=1, GravityOverride:Vector="0,0,-980", ExternalForce="0,0,0") -> (Pose) {
    angularDampingOverride=1
    angularSpringConstant=10000
    bAngularSpring=true
    bOverrideAngularDamping=true
    bUseGravityOverride=true
    gravityScale=0.35
    simulationSpace=BoneRelative
    alphaScaleBiasClamp {bInterpResult=true, interpSpeedDecreasing=1, interpSpeedIncreasing=5}
    boundBone {boneName=dyn_skirt_sd_r_03}
    componentLinearAccScale=<0.3 0.3 0.3>
    componentLinearVelScale=<0.3 0.3 0.3>
    gravityOverride {z=-980}
    physicsBodyDefinitions [{
      boundBone {boneName=dyn_skirt_sd_r_03}
      boxExtents=<30 20 20>
      constraintSetup {angularLimitsMax=<0 70 10>, angularLimitsMin=<0 -10 -10>, angularTarget=<1 0 0>}
      localJointOffset=<4 0 0>
     }]
    relativeSpaceBone {boneName=thigh_r}
    simSpaceSettings {simSpaceAngularAlpha=0.8}
   }
   n2 = Anim Dynamics - Bone: dyn_skirt_sd_r_01!(ComponentPose=@n1, Alpha=1, GravityOverride:Vector="0,0,-980", ExternalForce="0,0,0") -> (Pose) {
    angularDampingOverride=1
    angularSpringConstant=20000
    bAngularSpring=true
    bOverrideAngularDamping=true
    bUseGravityOverride=true
    gravityScale=0.35
    simulationSpace=RootRelative
    alphaScaleBiasClamp {bInterpResult=true, interpSpeedDecreasing=1, interpSpeedIncreasing=5}
    boundBone {boneName=dyn_skirt_sd_r_01}
    componentLinearAccScale=<0.3 0.3 0.3>
    componentLinearVelScale=<0.3 0.3 0.3>
    gravityOverride {z=-980}
    physicsBodyDefinitions [{
      boundBone {boneName=dyn_skirt_sd_r_01}
      boxExtents=<20 30 30>
      constraintSetup {angularLimitsMax=<0 0 10>, angularLimitsMin=<0 -50 -10>, angularTarget=<1 0 0>}
      localJointOffset=<4 0 0>
     }]
    simSpaceSettings {simSpaceAngularAlpha=0.8}
   }
   n3 = Anim Dynamics - Bone: dyn_skirt_sd_l_03!(ComponentPose=@n2, Alpha=1, GravityOverride:Vector="0,0,-980", ExternalForce="0,0,0") -> (Pose) {
    angularDampingOverride=1
    angularSpringConstant=10000
    bAngularSpring=true
    bOverrideAngularDamping=true
    bUseGravityOverride=true
    gravityScale=0.35
    simulationSpace=RootRelative
    alphaScaleBiasClamp {bInterpResult=true, interpSpeedDecreasing=1, interpSpeedIncreasing=5}
    boundBone {boneName=dyn_skirt_sd_l_03}
    componentLinearAccScale=<0.3 0.3 0.3>
    componentLinearVelScale=<0.3 0.3 0.3>
    gravityOverride {z=-980}
    physicsBodyDefinitions [{
      boundBone {boneName=dyn_skirt_sd_l_03}
      boxExtents=<30 20 20>
      constraintSetup {angularLimitsMax=<0 10 10>, angularLimitsMin=<0 -70 -10>, angularTarget=<1 0 0>}
      localJointOffset=<4 0 0>
     }]
    simSpaceSettings {simSpaceAngularAlpha=0.8}
   }
   n4 = Anim Dynamics - Bone: dyn_skirt_sd_l_02!(ComponentPose=@n3, Alpha=1, GravityOverride:Vector="0,0,-980", ExternalForce="0,0,0") -> (Pose) {
    angularDampingOverride=1
    angularSpringConstant=5000
    bAngularSpring=true
    bOverrideAngularDamping=true
    bUseGravityOverride=true
    gravityScale=0.35
    simulationSpace=RootRelative
    alphaScaleBiasClamp {bInterpResult=true, interpSpeedDecreasing=1, interpSpeedIncreasing=5}
    boundBone {boneName=dyn_skirt_sd_l_02}
    componentLinearAccScale=<0.3 0.3 0.3>
    componentLinearVelScale=<0.3 0.3 0.3>
    gravityOverride {z=-980}
    physicsBodyDefinitions [{
      boundBone {boneName=dyn_skirt_sd_l_02}
      boxExtents=<20 30 30>
      constraintSetup {angularLimitsMax=<0 0 5>, angularLimitsMin=<0 -50 -5>, angularTarget=<1 0 0>}
      localJointOffset=<4 0 0>
     }]
    simSpaceSettings {simSpaceAngularAlpha=0.8}
   }
   n5 = Anim Dynamics - Bone: dyn_skirt_sd_l_01!(ComponentPose=@n4, Alpha=1, GravityOverride:Vector="0,0,-980", ExternalForce="0,0,0") -> (Pose) {
    angularDampingOverride=1
    angularSpringConstant=20000
    bAngularSpring=true
    bOverrideAngularDamping=true
    bUseGravityOverride=true
    gravityScale=0.35
    simulationSpace=RootRelative
    alphaScaleBiasClamp {bInterpResult=true, interpSpeedDecreasing=1, interpSpeedIncreasing=5}
    boundBone {boneName=dyn_skirt_sd_l_01}
    componentLinearAccScale=<0.3 0.3 0.3>
    componentLinearVelScale=<0.3 0.3 0.3>
    gravityOverride {z=-980}
    physicsBodyDefinitions [{
      boundBone {boneName=dyn_skirt_sd_l_01}
      boxExtents=<20 30 30>
      constraintSetup {angularLimitsMax=<0 0 10>, angularLimitsMin=<0 -50 -10>, angularTarget=<1 0 0>}
      localJointOffset=<4 0 0>
     }]
    simSpaceSettings {simSpaceAngularAlpha=0.8}
   }
   n6 = Anim Dynamics - Bone: dyn_skirt_fr_r_03!(ComponentPose=@n5, Alpha=1, GravityOverride:Vector="0,0,-980", ExternalForce="0,0,0") -> (Pose) {
    angularDampingOverride=1
    angularSpringConstant=10000
    bAngularSpring=true
    bOverrideAngularDamping=true
    bUseGravityOverride=true
    gravityScale=0.35
    simulationSpace=RootRelative
    alphaScaleBiasClamp {bInterpResult=true, interpSpeedDecreasing=1, interpSpeedIncreasing=5}
    boundBone {boneName=dyn_skirt_fr_r_03}
    componentLinearAccScale=<0.3 0.3 0.3>
    componentLinearVelScale=<0.3 0.3 0.3>
    gravityOverride {z=-980}
    physicsBodyDefinitions [{
      boundBone {boneName=dyn_skirt_fr_r_03}
      boxExtents=<30 20 20>
      constraintSetup {angularLimitsMax=<0 10 0>, angularLimitsMin=<0 -70 0>, angularTarget=<1 0 0>}
      localJointOffset=<4 0 0>
     }]
    simSpaceSettings {simSpaceAngularAlpha=0.6}
   }
   n7 = Anim Dynamics - Bone: dyn_skirt_fr_r_02!(ComponentPose=@n6, Alpha=1, GravityOverride:Vector="0,0,-980", ExternalForce="0,0,0") -> (Pose) {
    angularDampingOverride=1
    angularSpringConstant=5000
    bAngularSpring=true
    bOverrideAngularDamping=true
    bUseGravityOverride=true
    gravityScale=0.35
    simulationSpace=RootRelative
    alphaScaleBiasClamp {bInterpResult=true, interpSpeedDecreasing=1, interpSpeedIncreasing=5}
    boundBone {boneName=dyn_skirt_fr_r_02}
    componentLinearAccScale=<0.3 0.3 0.3>
    componentLinearVelScale=<0.3 0.3 0.3>
    gravityOverride {z=-980}
    physicsBodyDefinitions [{
      boundBone {boneName=dyn_skirt_fr_r_02}
      boxExtents=<20 30 30>
      constraintSetup {angularLimitsMax=<0 0 0>, angularLimitsMin=<0 -30 0>, angularTarget=<1 0 0>}
      localJointOffset=<4 0 0>
     }]
    simSpaceSettings {simSpaceAngularAlpha=0.6}
   }
   n8 = Anim Dynamics - Bone: dyn_skirt_fr_r_01!(ComponentPose=@n7, Alpha=1, GravityOverride:Vector="0,0,-980", ExternalForce="0,0,0") -> (Pose) {
    angularDampingOverride=1
    angularSpringConstant=20000
    bAngularSpring=true
    bOverrideAngularDamping=true
    bUseGravityOverride=true
    gravityScale=0.35
    simulationSpace=RootRelative
    alphaScaleBiasClamp {bInterpResult=true, interpSpeedDecreasing=1, interpSpeedIncreasing=5}
    boundBone {boneName=dyn_skirt_fr_r_01}
    componentLinearAccScale=<0.3 0.3 0.3>
    componentLinearVelScale=<0.3 0.3 0.3>
    gravityOverride {z=-980}
    physicsBodyDefinitions [{
      boundBone {boneName=dyn_skirt_fr_r_01}
      boxExtents=<20 30 30>
      constraintSetup {angularLimitsMax=<0 0 10>, angularLimitsMin=<0 -10 -10>, angularTarget=<1 0 0>}
      localJointOffset=<4 0 0>
     }]
    simSpaceSettings {simSpaceAngularAlpha=0.6}
   }
   n9 = Anim Dynamics - Bone: dyn_skirt_fr_c_03!(ComponentPose=@n8, Alpha=1, GravityOverride:Vector="0,0,-980", ExternalForce="0,0,0") -> (Pose) {
    angularDampingOverride=1
    angularSpringConstant=5000
    bAngularSpring=true
    bOverrideAngularDamping=true
    bUseGravityOverride=true
    gravityScale=0.35
    simulationSpace=BoneRelative
    alphaScaleBiasClamp {bInterpResult=true, interpSpeedDecreasing=1, interpSpeedIncreasing=5}
    boundBone {boneName=dyn_skirt_fr_c_03}
    componentLinearAccScale=<0.3 0.3 0.3>
    componentLinearVelScale=<0.3 0.3 0.3>
    gravityOverride {z=-980}
    physicsBodyDefinitions [{
      boundBone {boneName=dyn_skirt_fr_c_03}
      boxExtents=<20 30 30>
      constraintSetup {angularLimitsMax=<0 0 8>, angularLimitsMin=<0 -5 -8>, angularTarget=<1 0 0>}
      localJointOffset=<4 0 0>
     }]
    relativeSpaceBone {boneName=thigh_r}
    simSpaceSettings {maxAngularAcceleration=5000, maxAngularVelocity=5000, simSpaceAngularAlpha=0.1}
   }
   n10 = Anim Dynamics - Bone: dyn_skirt_fr_l_03!(ComponentPose=@n9, Alpha=1, GravityOverride:Vector="0,0,-980", ExternalForce="0,0,0") -> (Pose) {
    angularDampingOverride=1
    angularSpringConstant=10000
    bAngularSpring=true
    bOverrideAngularDamping=true
    bUseGravityOverride=true
    gravityScale=0.35
    simulationSpace=BoneRelative
    alphaScaleBiasClamp {bInterpResult=true, interpSpeedDecreasing=1, interpSpeedIncreasing=5}
    boundBone {boneName=dyn_skirt_fr_l_03}
    componentLinearAccScale=<0.3 0.3 0.3>
    componentLinearVelScale=<0.3 0.3 0.3>
    gravityOverride {z=-980}
    physicsBodyDefinitions [{
      boundBone {boneName=dyn_skirt_fr_l_03}
      boxExtents=<30 20 20>
      constraintSetup {angularLimitsMax=<0 10 10>, angularLimitsMin=<0 -70 -10>, angularTarget=<1 0 0>}
      localJointOffset=<4 0 0>
     }]
    relativeSpaceBone {boneName=thigh_r}
    simSpaceSettings {simSpaceAngularAlpha=0.6}
   }
   n11 = Anim Dynamics - Bone: dyn_skirt_fr_l_01!(ComponentPose=@n10, Alpha=1, GravityOverride:Vector="0,0,-980", ExternalForce="0,0,0") -> (Pose) {
    angularDampingOverride=1
    angularSpringConstant=20000
    bAngularSpring=true
    bOverrideAngularDamping=true
    bUseGravityOverride=true
    gravityScale=0.35
    simulationSpace=RootRelative
    alphaScaleBiasClamp {bInterpResult=true, interpSpeedDecreasing=1, interpSpeedIncreasing=5}
    boundBone {boneName=dyn_skirt_fr_l_01}
    componentLinearAccScale=<0.3 0.3 0.3>
    componentLinearVelScale=<0.3 0.3 0.3>
    gravityOverride {z=-980}
    physicsBodyDefinitions [{
      boundBone {boneName=dyn_skirt_fr_l_01}
      boxExtents=<20 30 30>
      constraintSetup {angularLimitsMax=<0 0 10>, angularLimitsMin=<0 -10 -10>, angularTarget=<1 0 0>}
      localJointOffset=<4 0 0>
     }]
    simSpaceSettings {simSpaceAngularAlpha=0.6}
   }
   n12 = Anim Dynamics - Bone: dyn_skirt_bk_l_03!(ComponentPose=@n11, Alpha=1, GravityOverride:Vector="0,0,-980", ExternalForce="0,0,0") -> (Pose) {
    angularDampingOverride=1
    angularSpringConstant=5000
    bAngularSpring=true
    bOverrideAngularDamping=true
    bUseGravityOverride=true
    gravityScale=0.35
    simulationSpace=RootRelative
    alphaScaleBiasClamp {bInterpResult=true, interpSpeedDecreasing=1, interpSpeedIncreasing=5}
    boundBone {boneName=dyn_skirt_bk_l_03}
    componentAppliedLinearAccClamp=<200000 200000 200000>
    componentLinearAccScale=<0.3 0.3 0.3>
    componentLinearVelScale=<0.3 0.3 0.3>
    gravityOverride {z=-980}
    physicsBodyDefinitions [{
      boundBone {boneName=dyn_skirt_bk_l_03}
      boxExtents=<30 20 20>
      constraintSetup {angularLimitsMax=<0 10 10>, angularLimitsMin=<0 -30 -10>, angularTarget=<1 0 0>}
      localJointOffset=<4 0 0>
     }]
    simSpaceSettings {maxAngularAcceleration=30000, maxAngularVelocity=30000, simSpaceAngularAlpha=0.8, externalAngularVelocity=<0.3 0.3 0.3>}
   }
   n13 = Anim Dynamics - Bone: dyn_skirt_bk_l_02!(ComponentPose=@n12, Alpha=1, GravityOverride:Vector="0,0,-980", ExternalForce="0,0,0") -> (Pose) {
    angularDampingOverride=1
    angularSpringConstant=5000
    bAngularSpring=true
    bOverrideAngularDamping=true
    bUseGravityOverride=true
    gravityScale=0.35
    simulationSpace=RootRelative
    alphaScaleBiasClamp {bInterpResult=true, interpSpeedDecreasing=1, interpSpeedIncreasing=5}
    boundBone {boneName=dyn_skirt_bk_l_02}
    componentAppliedLinearAccClamp=<200000 200000 200000>
    componentLinearAccScale=<0.3 0.3 0.3>
    componentLinearVelScale=<0.3 0.3 0.3>
    gravityOverride {z=-980}
    physicsBodyDefinitions [{
      boundBone {boneName=dyn_skirt_bk_l_02}
      boxExtents=<30 20 20>
      constraintSetup {angularLimitsMax=<0 0 0>, angularLimitsMin=<0 -30 0>, angularTarget=<1 0 0>}
      localJointOffset=<4 0 0>
     }]
    simSpaceSettings {maxAngularAcceleration=30000, maxAngularVelocity=30000, simSpaceAngularAlpha=0.8, externalAngularVelocity=<0.3 0.3 0.3>}
   }
   n14 = Anim Dynamics - Bone: dyn_skirt_bk_l_01!(ComponentPose=@n13, Alpha=1, GravityOverride:Vector="0,0,-980", ExternalForce="0,0,0") -> (Pose) {
    angularDampingOverride=1
    angularSpringConstant=5000
    bAngularSpring=true
    bOverrideAngularDamping=true
    bUseGravityOverride=true
    gravityScale=0.35
    simulationSpace=RootRelative
    alphaScaleBiasClamp {bInterpResult=true, interpSpeedDecreasing=1, interpSpeedIncreasing=5}
    boundBone {boneName=dyn_skirt_bk_l_01}
    componentAppliedLinearAccClamp=<200000 200000 200000>
    componentLinearAccScale=<0.3 0.3 0.3>
    componentLinearVelScale=<0.3 0.3 0.3>
    gravityOverride {z=-980}
    physicsBodyDefinitions [{
      boundBone {boneName=dyn_skirt_bk_l_01}
      boxExtents=<30 20 20>
      constraintSetup {angularLimitsMax=<0 0 10>, angularLimitsMin=<0 -70 -10>, angularTarget=<1 0 0>}
      localJointOffset=<5 0 0>
     }]
    simSpaceSettings {maxAngularAcceleration=30000, maxAngularVelocity=30000, simSpaceAngularAlpha=1, externalAngularVelocity=<0.3 0.3 0.3>}
   }
   n15 = Anim Dynamics - Bone: dyn_skirt_bk_c_03!(ComponentPose=@n14, Alpha=1, GravityOverride:Vector="0,0,-980", ExternalForce="0,0,0") -> (Pose) {
    angularDampingOverride=1
    angularSpringConstant=5000
    bAngularSpring=true
    bOverrideAngularDamping=true
    bUseGravityOverride=true
    gravityScale=0.35
    simulationSpace=RootRelative
    alphaScaleBiasClamp {bInterpResult=true, interpSpeedDecreasing=1, interpSpeedIncreasing=5}
    boundBone {boneName=dyn_skirt_bk_c_03}
    componentAppliedLinearAccClamp=<200000 200000 200000>
    componentLinearAccScale=<0.3 0.3 0.3>
    componentLinearVelScale=<0.3 0.3 0.3>
    gravityOverride {z=-980}
    physicsBodyDefinitions [{
      boundBone {boneName=dyn_skirt_bk_c_03}
      boxExtents=<30 20 20>
      constraintSetup {angularLimitsMax=<0 10 10>, angularLimitsMin=<0 -30 -5>, angularTarget=<1 0 0>}
      localJointOffset=<4 0 0>
     }]
    simSpaceSettings {maxAngularAcceleration=30000, maxAngularVelocity=30000, simSpaceAngularAlpha=0.8, externalAngularVelocity=<0.3 0.3 0.3>}
   }
   n16 = Anim Dynamics - Bone: dyn_skirt_bk_c_02!(ComponentPose=@n15, Alpha=1, GravityOverride:Vector="0,0,-980", ExternalForce="0,0,0") -> (Pose) {
    angularDampingOverride=1
    angularSpringConstant=2000
    bAngularSpring=true
    bOverrideAngularDamping=true
    bUseGravityOverride=true
    gravityScale=0.35
    simulationSpace=RootRelative
    alphaScaleBiasClamp {bInterpResult=true, interpSpeedDecreasing=1, interpSpeedIncreasing=5}
    boundBone {boneName=dyn_skirt_bk_c_02}
    componentAppliedLinearAccClamp=<200000 200000 200000>
    componentLinearAccScale=<0.3 0.3 0.3>
    componentLinearVelScale=<0.3 0.3 0.3>
    gravityOverride {z=-980}
    physicsBodyDefinitions [{
      boundBone {boneName=dyn_skirt_bk_c_02}
      boxExtents=<30 20 20>
      constraintSetup {angularLimitsMax=<0 0 0>, angularLimitsMin=<0 -20 0>, angularTarget=<1 0 0>}
      localJointOffset=<4 0 0>
     }]
    simSpaceSettings {maxAngularAcceleration=30000, maxAngularVelocity=30000, simSpaceAngularAlpha=0.8, externalAngularVelocity=<0.3 0.3 0.3>}
   }
   n17 = Anim Dynamics - Bone: dyn_skirt_bk_c_01!(ComponentPose=@n16, Alpha=1, GravityOverride:Vector="0,0,-980", ExternalForce="0,0,0") -> (Pose) {
    angularDampingOverride=1
    angularSpringConstant=5000
    bAngularSpring=true
    bOverrideAngularDamping=true
    bUseGravityOverride=true
    gravityScale=0.35
    simulationSpace=RootRelative
    alphaScaleBiasClamp {bInterpResult=true, interpSpeedDecreasing=1, interpSpeedIncreasing=5}
    boundBone {boneName=dyn_skirt_bk_c_01}
    componentAppliedLinearAccClamp=<200000 200000 200000>
    componentLinearAccScale=<0.3 0.3 0.3>
    componentLinearVelScale=<0.3 0.3 0.3>
    gravityOverride {z=-980}
    physicsBodyDefinitions [{
      boundBone {boneName=dyn_skirt_bk_c_01}
      boxExtents=<30 20 20>
      constraintSetup {angularLimitsMax=<0 0 10>, angularLimitsMin=<0 -80 -2>, angularTarget=<1 0 0>}
      localJointOffset=<5 0 0>
     }]
    simSpaceSettings {maxAngularAcceleration=30000, maxAngularVelocity=30000, simSpaceAngularAlpha=1, externalAngularVelocity=<0.3 0.3 0.3>}
   }
   n18 = Anim Dynamics - Bone: dyn_skirt_bk_r_01!(ComponentPose=@n17, Alpha=1, GravityOverride:Vector="0,0,-980", ExternalForce="0,0,0") -> (Pose) {
    angularDampingOverride=1
    angularSpringConstant=5000
    bAngularSpring=true
    bOverrideAngularDamping=true
    bUseGravityOverride=true
    gravityScale=0.35
    simulationSpace=RootRelative
    alphaScaleBiasClamp {bInterpResult=true, interpSpeedDecreasing=1, interpSpeedIncreasing=5}
    boundBone {boneName=dyn_skirt_bk_r_01}
    componentAppliedLinearAccClamp=<200000 200000 200000>
    componentLinearAccScale=<0.3 0.3 0.3>
    componentLinearVelScale=<0.3 0.3 0.3>
    gravityOverride {z=-980}
    physicsBodyDefinitions [{
      boundBone {boneName=dyn_skirt_bk_r_01}
      boxExtents=<30 20 20>
      constraintSetup {angularLimitsMax=<0 0 10>, angularLimitsMin=<0 -70 -10>, angularTarget=<1 0 0>}
      localJointOffset=<5 0 0>
     }]
    simSpaceSettings {maxAngularAcceleration=30000, maxAngularVelocity=30000, simSpaceAngularAlpha=1, externalAngularVelocity=<0.3 0.3 0.3>}
   }
   n19 = Anim Dynamics - Bone: dyn_skirt_bk_r_02!(ComponentPose=@n18, Alpha=1, GravityOverride:Vector="0,0,-980", ExternalForce="0,0,0") -> (Pose) {
    angularDampingOverride=1
    angularSpringConstant=5000
    bAngularSpring=true
    bOverrideAngularDamping=true
    bUseGravityOverride=true
    gravityScale=0.35
    simulationSpace=RootRelative
    alphaScaleBiasClamp {bInterpResult=true, interpSpeedDecreasing=1, interpSpeedIncreasing=5}
    boundBone {boneName=dyn_skirt_bk_r_02}
    componentAppliedLinearAccClamp=<200000 200000 200000>
    componentLinearAccScale=<0.3 0.3 0.3>
    componentLinearVelScale=<0.3 0.3 0.3>
    gravityOverride {z=-980}
    physicsBodyDefinitions [{
      boundBone {boneName=dyn_skirt_bk_r_02}
      boxExtents=<30 20 20>
      constraintSetup {angularLimitsMax=<0 0 5>, angularLimitsMin=<0 -70 -5>, angularTarget=<1 0 0>}
      localJointOffset=<4 0 0>
     }]
    simSpaceSettings {maxAngularAcceleration=30000, maxAngularVelocity=30000, simSpaceAngularAlpha=0.8, externalAngularVelocity=<0.3 0.3 0.3>}
   }
   n20 = Anim Dynamics - Bone: dyn_skirt_bk_r_03!(ComponentPose=@n19, Alpha=1, GravityOverride:Vector="0,0,-980", ExternalForce="0,0,0") -> (Pose) {
    angularDampingOverride=1
    angularSpringConstant=5000
    bAngularSpring=true
    bOverrideAngularDamping=true
    bUseGravityOverride=true
    gravityScale=0.35
    simulationSpace=RootRelative
    alphaScaleBiasClamp {bInterpResult=true, interpSpeedDecreasing=1, interpSpeedIncreasing=5}
    boundBone {boneName=dyn_skirt_bk_r_03}
    componentAppliedLinearAccClamp=<200000 200000 200000>
    componentLinearAccScale=<0.3 0.3 0.3>
    componentLinearVelScale=<0.3 0.3 0.3>
    gravityOverride {z=-980}
    physicsBodyDefinitions [{
      boundBone {boneName=dyn_skirt_bk_r_03}
      boxExtents=<30 20 20>
      constraintSetup {angularLimitsMax=<0 10 10>, angularLimitsMin=<0 -80 -10>, angularTarget=<1 0 0>}
      localJointOffset=<4 0 0>
     }]
    simSpaceSettings {maxAngularAcceleration=30000, maxAngularVelocity=30000, simSpaceAngularAlpha=0.8, externalAngularVelocity=<0.3 0.3 0.3>}
   }
   n21 = Anim Dynamics - Bone: dyn_skirt_sd_r_03(ComponentPose=Local To Component(LocalPose=Use cached pose 'start'()), Alpha=1, GravityOverride="0,0,-980", ExternalForce="0,0,0") {
    angularDampingOverride=1
    angularSpringConstant=5000
    bAngularSpring=true
    bOverrideAngularDamping=true
    bUseGravityOverride=true
    gravityScale=0.35
    simulationSpace=BoneRelative
    alphaScaleBiasClamp {bInterpResult=true, interpSpeedDecreasing=1, interpSpeedIncreasing=5}
    boundBone {boneName=dyn_skirt_sd_r_03}
    componentLinearAccScale=<0.3 0.3 0.3>
    componentLinearVelScale=<0.3 0.3 0.3>
    gravityOverride {z=-980}
    physicsBodyDefinitions [{
      boundBone {boneName=dyn_skirt_sd_r_03}
      boxExtents=<30 20 20>
      constraintSetup {angularLimitsMax=<0 70 10>, angularLimitsMin=<0 -10 -10>, angularTarget=<1 0 0>}
      localJointOffset=<4 0 0>
     }]
    relativeSpaceBone {boneName=thigh_r}
    simSpaceSettings {simSpaceAngularAlpha=0.2}
   }
   n22 = Anim Dynamics - Bone: dyn_skirt_sd_r_01(ComponentPose=@n21, Alpha=1, GravityOverride="0,0,-980", ExternalForce="0,0,0") {
    angularDampingOverride=1
    angularSpringConstant=5000
    bAngularSpring=true
    bOverrideAngularDamping=true
    bUseGravityOverride=true
    gravityScale=0.35
    simulationSpace=RootRelative
    alphaScaleBiasClamp {bInterpResult=true, interpSpeedDecreasing=1, interpSpeedIncreasing=5}
    boundBone {boneName=dyn_skirt_sd_r_01}
    componentLinearAccScale=<0.3 0.3 0.3>
    componentLinearVelScale=<0.3 0.3 0.3>
    gravityOverride {z=-980}
    physicsBodyDefinitions [{
      boundBone {boneName=dyn_skirt_sd_r_01}
      boxExtents=<20 30 30>
      constraintSetup {angularLimitsMax=<0 0 10>, angularLimitsMin=<0 -50 -10>, angularTarget=<1 0 0>}
      localJointOffset=<4 0 0>
     }]
    simSpaceSettings {simSpaceAngularAlpha=0.2}
   }
   n23 = Anim Dynamics - Bone: dyn_skirt_sd_l_03(ComponentPose=@n22, Alpha=1, GravityOverride="0,0,-980", ExternalForce="0,0,0") {
    angularDampingOverride=1
    angularSpringConstant=5000
    bAngularSpring=true
    bOverrideAngularDamping=true
    bUseGravityOverride=true
    gravityScale=0.35
    simulationSpace=RootRelative
    alphaScaleBiasClamp {bInterpResult=true, interpSpeedDecreasing=1, interpSpeedIncreasing=5}
    boundBone {boneName=dyn_skirt_sd_l_03}
    componentLinearAccScale=<0.3 0.3 0.3>
    componentLinearVelScale=<0.3 0.3 0.3>
    gravityOverride {z=-980}
    physicsBodyDefinitions [{
      boundBone {boneName=dyn_skirt_sd_l_03}
      boxExtents=<30 20 20>
      constraintSetup {angularLimitsMax=<0 10 10>, angularLimitsMin=<0 -70 -10>, angularTarget=<1 0 0>}
      localJointOffset=<4 0 0>
     }]
    simSpaceSettings {simSpaceAngularAlpha=0.2}
   }
   n24 = Anim Dynamics - Bone: dyn_skirt_sd_l_02(ComponentPose=@n23, Alpha=1, GravityOverride="0,0,-980", ExternalForce="0,0,0") {
    angularDampingOverride=1
    angularSpringConstant=5000
    bAngularSpring=true
    bOverrideAngularDamping=true
    bUseGravityOverride=true
    gravityScale=0.35
    simulationSpace=RootRelative
    alphaScaleBiasClamp {bInterpResult=true, interpSpeedDecreasing=1, interpSpeedIncreasing=5}
    boundBone {boneName=dyn_skirt_sd_l_02}
    componentLinearAccScale=<0.3 0.3 0.3>
    componentLinearVelScale=<0.3 0.3 0.3>
    gravityOverride {z=-980}
    physicsBodyDefinitions [{
      boundBone {boneName=dyn_skirt_sd_l_02}
      boxExtents=<20 30 30>
      constraintSetup {angularLimitsMax=<0 0 5>, angularLimitsMin=<0 -50 -5>, angularTarget=<1 0 0>}
      localJointOffset=<4 0 0>
     }]
    simSpaceSettings {simSpaceAngularAlpha=0.2}
   }
   n25 = Anim Dynamics - Bone: dyn_skirt_sd_l_01(ComponentPose=@n24, Alpha=1, GravityOverride="0,0,-980", ExternalForce="0,0,0") {
    angularDampingOverride=1
    angularSpringConstant=5000
    bAngularSpring=true
    bOverrideAngularDamping=true
    bUseGravityOverride=true
    gravityScale=0.35
    simulationSpace=RootRelative
    alphaScaleBiasClamp {bInterpResult=true, interpSpeedDecreasing=1, interpSpeedIncreasing=5}
    boundBone {boneName=dyn_skirt_sd_l_01}
    componentLinearAccScale=<0.3 0.3 0.3>
    componentLinearVelScale=<0.3 0.3 0.3>
    gravityOverride {z=-980}
    physicsBodyDefinitions [{
      boundBone {boneName=dyn_skirt_sd_l_01}
      boxExtents=<20 30 30>
      constraintSetup {angularLimitsMax=<0 0 10>, angularLimitsMin=<0 -50 -10>, angularTarget=<1 0 0>}
      localJointOffset=<4 0 0>
     }]
    simSpaceSettings {simSpaceAngularAlpha=0.2}
   }
   n26 = Anim Dynamics - Bone: dyn_skirt_fr_r_03(ComponentPose=@n25, Alpha=1, GravityOverride="0,0,-980", ExternalForce="0,0,0") {
    angularDampingOverride=1
    angularSpringConstant=5000
    bAngularSpring=true
    bOverrideAngularDamping=true
    bUseGravityOverride=true
    gravityScale=0.35
    simulationSpace=RootRelative
    alphaScaleBiasClamp {bInterpResult=true, interpSpeedDecreasing=1, interpSpeedIncreasing=5}
    boundBone {boneName=dyn_skirt_fr_r_03}
    componentLinearAccScale=<0.3 0.3 0.3>
    componentLinearVelScale=<0.3 0.3 0.3>
    gravityOverride {z=-980}
    physicsBodyDefinitions [{
      boundBone {boneName=dyn_skirt_fr_r_03}
      boxExtents=<30 20 20>
      constraintSetup {angularLimitsMax=<0 10 0>, angularLimitsMin=<0 -70 0>, angularTarget=<1 0 0>}
      localJointOffset=<4 0 0>
     }]
    simSpaceSettings {simSpaceAngularAlpha=0.2}
   }
   n27 = Anim Dynamics - Bone: dyn_skirt_fr_r_02(ComponentPose=@n26, Alpha=1, GravityOverride="0,0,-980", ExternalForce="0,0,0") {
    angularDampingOverride=1
    angularSpringConstant=5000
    bAngularSpring=true
    bOverrideAngularDamping=true
    bUseGravityOverride=true
    gravityScale=0.35
    simulationSpace=RootRelative
    alphaScaleBiasClamp {bInterpResult=true, interpSpeedDecreasing=1, interpSpeedIncreasing=5}
    boundBone {boneName=dyn_skirt_fr_r_02}
    componentLinearAccScale=<0.3 0.3 0.3>
    componentLinearVelScale=<0.3 0.3 0.3>
    gravityOverride {z=-980}
    physicsBodyDefinitions [{
      boundBone {boneName=dyn_skirt_fr_r_02}
      boxExtents=<20 30 30>
      constraintSetup {angularLimitsMax=<0 0 0>, angularLimitsMin=<0 -30 0>, angularTarget=<1 0 0>}
      localJointOffset=<4 0 0>
     }]
    simSpaceSettings {simSpaceAngularAlpha=0.2}
   }
   n28 = Anim Dynamics - Bone: dyn_skirt_fr_r_01(ComponentPose=@n27, Alpha=1, GravityOverride="0,0,-980", ExternalForce="0,0,0") {
    angularDampingOverride=1
    angularSpringConstant=5000
    bAngularSpring=true
    bOverrideAngularDamping=true
    bUseGravityOverride=true
    gravityScale=0.35
    simulationSpace=RootRelative
    alphaScaleBiasClamp {bInterpResult=true, interpSpeedDecreasing=1, interpSpeedIncreasing=5}
    boundBone {boneName=dyn_skirt_fr_r_01}
    componentLinearAccScale=<0.3 0.3 0.3>
    componentLinearVelScale=<0.3 0.3 0.3>
    gravityOverride {z=-980}
    physicsBodyDefinitions [{
      boundBone {boneName=dyn_skirt_fr_r_01}
      boxExtents=<20 30 30>
      constraintSetup {angularLimitsMax=<0 0 10>, angularLimitsMin=<0 -10 -10>, angularTarget=<1 0 0>}
      localJointOffset=<4 0 0>
     }]
    simSpaceSettings {simSpaceAngularAlpha=0.2}
   }
   n29 = Anim Dynamics - Bone: dyn_skirt_fr_c_03(ComponentPose=@n28, Alpha=1, GravityOverride="0,0,-980", ExternalForce="0,0,0") {
    angularDampingOverride=1
    angularSpringConstant=5000
    bAngularSpring=true
    bOverrideAngularDamping=true
    bUseGravityOverride=true
    gravityScale=0.35
    simulationSpace=BoneRelative
    alphaScaleBiasClamp {bInterpResult=true, interpSpeedDecreasing=1, interpSpeedIncreasing=5}
    boundBone {boneName=dyn_skirt_fr_c_03}
    componentLinearAccScale=<0.3 0.3 0.3>
    componentLinearVelScale=<0.3 0.3 0.3>
    gravityOverride {z=-980}
    physicsBodyDefinitions [{
      boundBone {boneName=dyn_skirt_fr_c_03}
      boxExtents=<20 30 30>
      constraintSetup {angularLimitsMax=<0 0 8>, angularLimitsMin=<0 -5 -8>, angularTarget=<1 0 0>}
      localJointOffset=<4 0 0>
     }]
    relativeSpaceBone {boneName=thigh_r}
    simSpaceSettings {maxAngularAcceleration=5000, maxAngularVelocity=5000, simSpaceAngularAlpha=0.2}
   }
   n30 = Anim Dynamics - Bone: dyn_skirt_fr_l_03(ComponentPose=@n29, Alpha=1, GravityOverride="0,0,-980", ExternalForce="0,0,0") {
    angularDampingOverride=1
    angularSpringConstant=5000
    bAngularSpring=true
    bOverrideAngularDamping=true
    bUseGravityOverride=true
    gravityScale=0.35
    simulationSpace=BoneRelative
    alphaScaleBiasClamp {bInterpResult=true, interpSpeedDecreasing=1, interpSpeedIncreasing=5}
    boundBone {boneName=dyn_skirt_fr_l_03}
    componentLinearAccScale=<0.3 0.3 0.3>
    componentLinearVelScale=<0.3 0.3 0.3>
    gravityOverride {z=-980}
    physicsBodyDefinitions [{
      boundBone {boneName=dyn_skirt_fr_l_03}
      boxExtents=<30 20 20>
      constraintSetup {angularLimitsMax=<0 10 10>, angularLimitsMin=<0 -70 -10>, angularTarget=<1 0 0>}
      localJointOffset=<4 0 0>
     }]
    relativeSpaceBone {boneName=thigh_r}
    simSpaceSettings {simSpaceAngularAlpha=0.2}
   }
   n31 = Anim Dynamics - Bone: dyn_skirt_fr_l_01(ComponentPose=@n30, Alpha=1, GravityOverride="0,0,-980", ExternalForce="0,0,0") {
    angularDampingOverride=1
    angularSpringConstant=5000
    bAngularSpring=true
    bOverrideAngularDamping=true
    bUseGravityOverride=true
    gravityScale=0.35
    simulationSpace=RootRelative
    alphaScaleBiasClamp {bInterpResult=true, interpSpeedDecreasing=1, interpSpeedIncreasing=5}
    boundBone {boneName=dyn_skirt_fr_l_01}
    componentLinearAccScale=<0.3 0.3 0.3>
    componentLinearVelScale=<0.3 0.3 0.3>
    gravityOverride {z=-980}
    physicsBodyDefinitions [{
      boundBone {boneName=dyn_skirt_fr_l_01}
      boxExtents=<20 30 30>
      constraintSetup {angularLimitsMax=<0 0 10>, angularLimitsMin=<0 -10 -10>, angularTarget=<1 0 0>}
      localJointOffset=<4 0 0>
     }]
    simSpaceSettings {simSpaceAngularAlpha=0.2}
   }
   n32 = Anim Dynamics - Bone: dyn_skirt_bk_l_03(ComponentPose=@n31, Alpha=1, GravityOverride="0,0,-980", ExternalForce="0,0,0") {
    angularDampingOverride=1
    angularSpringConstant=5000
    bAngularSpring=true
    bOverrideAngularDamping=true
    bUseGravityOverride=true
    gravityScale=0.35
    simulationSpace=RootRelative
    alphaScaleBiasClamp {bInterpResult=true, interpSpeedDecreasing=1, interpSpeedIncreasing=5}
    boundBone {boneName=dyn_skirt_bk_l_03}
    componentAppliedLinearAccClamp=<200000 200000 200000>
    componentLinearAccScale=<0.3 0.3 0.3>
    componentLinearVelScale=<0.3 0.3 0.3>
    gravityOverride {z=-980}
    physicsBodyDefinitions [{
      boundBone {boneName=dyn_skirt_bk_l_03}
      boxExtents=<30 20 20>
      constraintSetup {angularLimitsMax=<0 10 10>, angularLimitsMin=<0 -30 -10>, angularTarget=<1 0 0>}
      localJointOffset=<4 0 0>
     }]
    simSpaceSettings {maxAngularAcceleration=30000, maxAngularVelocity=30000, simSpaceAngularAlpha=0.2, externalAngularVelocity=<0.3 0.3 0.3>}
   }
   n33 = Anim Dynamics - Bone: dyn_skirt_bk_l_02(ComponentPose=@n32, Alpha=1, GravityOverride="0,0,-980", ExternalForce="0,0,0") {
    angularDampingOverride=1
    angularSpringConstant=5000
    bAngularSpring=true
    bOverrideAngularDamping=true
    bUseGravityOverride=true
    gravityScale=0.35
    simulationSpace=RootRelative
    alphaScaleBiasClamp {bInterpResult=true, interpSpeedDecreasing=1, interpSpeedIncreasing=5}
    boundBone {boneName=dyn_skirt_bk_l_02}
    componentAppliedLinearAccClamp=<200000 200000 200000>
    componentLinearAccScale=<0.3 0.3 0.3>
    componentLinearVelScale=<0.3 0.3 0.3>
    gravityOverride {z=-980}
    physicsBodyDefinitions [{
      boundBone {boneName=dyn_skirt_bk_l_02}
      boxExtents=<30 20 20>
      constraintSetup {angularLimitsMax=<0 0 0>, angularLimitsMin=<0 -30 0>, angularTarget=<1 0 0>}
      localJointOffset=<4 0 0>
     }]
    simSpaceSettings {maxAngularAcceleration=30000, maxAngularVelocity=30000, simSpaceAngularAlpha=0.2, externalAngularVelocity=<0.3 0.3 0.3>}
   }
   n34 = Anim Dynamics - Bone: dyn_skirt_bk_l_01(ComponentPose=@n33, Alpha=1, GravityOverride="0,0,-980", ExternalForce="0,0,0") {
    angularDampingOverride=1
    angularSpringConstant=5000
    bAngularSpring=true
    bOverrideAngularDamping=true
    bUseGravityOverride=true
    gravityScale=0.35
    simulationSpace=RootRelative
    alphaScaleBiasClamp {bInterpResult=true, interpSpeedDecreasing=1, interpSpeedIncreasing=5}
    boundBone {boneName=dyn_skirt_bk_l_01}
    componentAppliedLinearAccClamp=<200000 200000 200000>
    componentLinearAccScale=<0.3 0.3 0.3>
    componentLinearVelScale=<0.3 0.3 0.3>
    gravityOverride {z=-980}
    physicsBodyDefinitions [{
      boundBone {boneName=dyn_skirt_bk_l_01}
      boxExtents=<30 20 20>
      constraintSetup {angularLimitsMax=<0 0 10>, angularLimitsMin=<0 -70 -10>, angularTarget=<1 0 0>}
      localJointOffset=<5 0 0>
     }]
    simSpaceSettings {maxAngularAcceleration=30000, maxAngularVelocity=30000, simSpaceAngularAlpha=0.2, externalAngularVelocity=<0.3 0.3 0.3>}
   }
   n35 = Anim Dynamics - Bone: dyn_skirt_bk_c_03(ComponentPose=@n34, Alpha=1, GravityOverride="0,0,-980", ExternalForce="0,0,0") {
    angularDampingOverride=1
    angularSpringConstant=5000
    bAngularSpring=true
    bOverrideAngularDamping=true
    bUseGravityOverride=true
    gravityScale=0.35
    simulationSpace=RootRelative
    alphaScaleBiasClamp {bInterpResult=true, interpSpeedDecreasing=1, interpSpeedIncreasing=5}
    boundBone {boneName=dyn_skirt_bk_c_03}
    componentAppliedLinearAccClamp=<200000 200000 200000>
    componentLinearAccScale=<0.3 0.3 0.3>
    componentLinearVelScale=<0.3 0.3 0.3>
    gravityOverride {z=-980}
    physicsBodyDefinitions [{
      boundBone {boneName=dyn_skirt_bk_c_03}
      boxExtents=<30 20 20>
      constraintSetup {angularLimitsMax=<0 10 10>, angularLimitsMin=<0 -30 -5>, angularTarget=<1 0 0>}
      localJointOffset=<4 0 0>
     }]
    simSpaceSettings {maxAngularAcceleration=30000, maxAngularVelocity=30000, simSpaceAngularAlpha=0.2, externalAngularVelocity=<0.3 0.3 0.3>}
   }
   n36 = Anim Dynamics - Bone: dyn_skirt_bk_c_02(ComponentPose=@n35, Alpha=1, GravityOverride="0,0,-980", ExternalForce="0,0,0") {
    angularDampingOverride=1
    angularSpringConstant=5000
    bAngularSpring=true
    bOverrideAngularDamping=true
    bUseGravityOverride=true
    gravityScale=0.35
    simulationSpace=RootRelative
    alphaScaleBiasClamp {bInterpResult=true, interpSpeedDecreasing=1, interpSpeedIncreasing=5}
    boundBone {boneName=dyn_skirt_bk_c_02}
    componentAppliedLinearAccClamp=<200000 200000 200000>
    componentLinearAccScale=<0.3 0.3 0.3>
    componentLinearVelScale=<0.3 0.3 0.3>
    gravityOverride {z=-980}
    physicsBodyDefinitions [{
      boundBone {boneName=dyn_skirt_bk_c_02}
      boxExtents=<30 20 20>
      constraintSetup {angularLimitsMax=<0 0 0>, angularLimitsMin=<0 -20 0>, angularTarget=<1 0 0>}
      localJointOffset=<4 0 0>
     }]
    simSpaceSettings {maxAngularAcceleration=30000, maxAngularVelocity=30000, simSpaceAngularAlpha=0.2, externalAngularVelocity=<0.3 0.3 0.3>}
   }
   n37 = Anim Dynamics - Bone: dyn_skirt_bk_c_01(ComponentPose=@n36, Alpha=1, GravityOverride="0,0,-980", ExternalForce="0,0,0") {
    angularDampingOverride=1
    angularSpringConstant=5000
    bAngularSpring=true
    bOverrideAngularDamping=true
    bUseGravityOverride=true
    gravityScale=0.35
    simulationSpace=RootRelative
    alphaScaleBiasClamp {bInterpResult=true, interpSpeedDecreasing=1, interpSpeedIncreasing=5}
    boundBone {boneName=dyn_skirt_bk_c_01}
    componentAppliedLinearAccClamp=<200000 200000 200000>
    componentLinearAccScale=<0.3 0.3 0.3>
    componentLinearVelScale=<0.3 0.3 0.3>
    gravityOverride {z=-980}
    physicsBodyDefinitions [{
      boundBone {boneName=dyn_skirt_bk_c_01}
      boxExtents=<30 20 20>
      constraintSetup {angularLimitsMax=<0 0 10>, angularLimitsMin=<0 -80 -2>, angularTarget=<1 0 0>}
      localJointOffset=<5 0 0>
     }]
    simSpaceSettings {maxAngularAcceleration=30000, maxAngularVelocity=30000, simSpaceAngularAlpha=0.2, externalAngularVelocity=<0.3 0.3 0.3>}
   }
   n38 = Anim Dynamics - Bone: dyn_skirt_bk_r_01(ComponentPose=@n37, Alpha=1, GravityOverride="0,0,-980", ExternalForce="0,0,0") {
    angularDampingOverride=1
    angularSpringConstant=5000
    bAngularSpring=true
    bOverrideAngularDamping=true
    bUseGravityOverride=true
    gravityScale=0.35
    simulationSpace=RootRelative
    alphaScaleBiasClamp {bInterpResult=true, interpSpeedDecreasing=1, interpSpeedIncreasing=5}
    boundBone {boneName=dyn_skirt_bk_r_01}
    componentAppliedLinearAccClamp=<200000 200000 200000>
    componentLinearAccScale=<0.3 0.3 0.3>
    componentLinearVelScale=<0.3 0.3 0.3>
    gravityOverride {z=-980}
    physicsBodyDefinitions [{
      boundBone {boneName=dyn_skirt_bk_r_01}
      boxExtents=<30 20 20>
      constraintSetup {angularLimitsMax=<0 0 10>, angularLimitsMin=<0 -70 -10>, angularTarget=<1 0 0>}
      localJointOffset=<5 0 0>
     }]
    simSpaceSettings {maxAngularAcceleration=30000, maxAngularVelocity=30000, simSpaceAngularAlpha=0.2, externalAngularVelocity=<0.3 0.3 0.3>}
   }
   n39 = Anim Dynamics - Bone: dyn_skirt_bk_r_02(ComponentPose=@n38, Alpha=1, GravityOverride="0,0,-980", ExternalForce="0,0,0") {
    angularDampingOverride=1
    angularSpringConstant=5000
    bAngularSpring=true
    bOverrideAngularDamping=true
    bUseGravityOverride=true
    gravityScale=0.35
    simulationSpace=RootRelative
    alphaScaleBiasClamp {bInterpResult=true, interpSpeedDecreasing=1, interpSpeedIncreasing=5}
    boundBone {boneName=dyn_skirt_bk_r_02}
    componentAppliedLinearAccClamp=<200000 200000 200000>
    componentLinearAccScale=<0.3 0.3 0.3>
    componentLinearVelScale=<0.3 0.3 0.3>
    gravityOverride {z=-980}
    physicsBodyDefinitions [{
      boundBone {boneName=dyn_skirt_bk_r_02}
      boxExtents=<30 20 20>
      constraintSetup {angularLimitsMax=<0 0 5>, angularLimitsMin=<0 -70 -5>, angularTarget=<1 0 0>}
      localJointOffset=<4 0 0>
     }]
    simSpaceSettings {maxAngularAcceleration=30000, maxAngularVelocity=30000, simSpaceAngularAlpha=0.2, externalAngularVelocity=<0.3 0.3 0.3>}
   }
   n40 = Anim Dynamics - Bone: dyn_skirt_bk_r_03(ComponentPose=@n39, Alpha=1, GravityOverride="0,0,-980", ExternalForce="0,0,0") {
    angularDampingOverride=1
    angularSpringConstant=5000
    bAngularSpring=true
    bOverrideAngularDamping=true
    bUseGravityOverride=true
    gravityScale=0.35
    simulationSpace=RootRelative
    alphaScaleBiasClamp {bInterpResult=true, interpSpeedDecreasing=1, interpSpeedIncreasing=5}
    boundBone {boneName=dyn_skirt_bk_r_03}
    componentAppliedLinearAccClamp=<200000 200000 200000>
    componentLinearAccScale=<0.3 0.3 0.3>
    componentLinearVelScale=<0.3 0.3 0.3>
    gravityOverride {z=-980}
    physicsBodyDefinitions [{
      boundBone {boneName=dyn_skirt_bk_r_03}
      boxExtents=<30 20 20>
      constraintSetup {angularLimitsMax=<0 10 10>, angularLimitsMin=<0 -80 -10>, angularTarget=<1 0 0>}
      localJointOffset=<4 0 0>
     }]
    simSpaceSettings {maxAngularAcceleration=30000, maxAngularVelocity=30000, simSpaceAngularAlpha=0.2, externalAngularVelocity=<0.3 0.3 0.3>}
   }
   n41 = Anim Dynamics - Bone: dyn_skirt_sd_r_03(ComponentPose=Local To Component(LocalPose=Use cached pose 'start'()), Alpha=1, GravityOverride="0,0,-980", ExternalForce="0,0,0") {
    angularDampingOverride=1
    angularSpringConstant=10000
    bAngularSpring=true
    bOverrideAngularDamping=true
    bUseGravityOverride=true
    gravityScale=0.35
    simulationSpace=BoneRelative
    alphaScaleBiasClamp {bInterpResult=true, interpSpeedDecreasing=1, interpSpeedIncreasing=5}
    boundBone {boneName=dyn_skirt_sd_r_03}
    componentLinearAccScale=<0.3 0.3 0.3>
    componentLinearVelScale=<0.3 0.3 0.3>
    gravityOverride {z=-980}
    physicsBodyDefinitions [{
      boundBone {boneName=dyn_skirt_sd_r_03}
      boxExtents=<30 20 20>
      constraintSetup {angularLimitsMax=<0 70 10>, angularLimitsMin=<0 -10 -10>, angularTarget=<1 0 0>}
      localJointOffset=<4 0 0>
     }]
    relativeSpaceBone {boneName=thigh_r}
    simSpaceSettings {simSpaceAngularAlpha=0.8}
   }
   n42 = Anim Dynamics - Bone: dyn_skirt_sd_r_01(ComponentPose=@n41, Alpha=1, GravityOverride="0,0,-980", ExternalForce="0,0,0") {
    angularDampingOverride=1
    angularSpringConstant=20000
    bAngularSpring=true
    bOverrideAngularDamping=true
    bUseGravityOverride=true
    gravityScale=0.35
    simulationSpace=RootRelative
    alphaScaleBiasClamp {bInterpResult=true, interpSpeedDecreasing=1, interpSpeedIncreasing=5}
    boundBone {boneName=dyn_skirt_sd_r_01}
    componentLinearAccScale=<0.3 0.3 0.3>
    componentLinearVelScale=<0.3 0.3 0.3>
    gravityOverride {z=-980}
    physicsBodyDefinitions [{
      boundBone {boneName=dyn_skirt_sd_r_01}
      boxExtents=<20 30 30>
      constraintSetup {angularLimitsMax=<0 0 10>, angularLimitsMin=<0 -10 -10>, angularTarget=<1 0 0>}
      localJointOffset=<4 0 0>
     }]
    simSpaceSettings {simSpaceAngularAlpha=0.6}
   }
   n43 = Anim Dynamics - Bone: dyn_skirt_sd_l_03(ComponentPose=@n42, Alpha=1, GravityOverride="0,0,-980", ExternalForce="0,0,0") {
    angularDampingOverride=1
    angularSpringConstant=10000
    bAngularSpring=true
    bOverrideAngularDamping=true
    bUseGravityOverride=true
    gravityScale=0.35
    simulationSpace=RootRelative
    alphaScaleBiasClamp {bInterpResult=true, interpSpeedDecreasing=1, interpSpeedIncreasing=5}
    boundBone {boneName=dyn_skirt_sd_l_03}
    componentLinearAccScale=<0.3 0.3 0.3>
    componentLinearVelScale=<0.3 0.3 0.3>
    gravityOverride {z=-980}
    physicsBodyDefinitions [{
      boundBone {boneName=dyn_skirt_sd_l_03}
      boxExtents=<30 20 20>
      constraintSetup {angularLimitsMax=<0 10 10>, angularLimitsMin=<0 -70 -10>, angularTarget=<1 0 0>}
      localJointOffset=<4 0 0>
     }]
    simSpaceSettings {simSpaceAngularAlpha=0.8}
   }
   n44 = Anim Dynamics - Bone: dyn_skirt_sd_l_02(ComponentPose=@n43, Alpha=1, GravityOverride="0,0,-980", ExternalForce="0,0,0") {
    angularDampingOverride=1
    angularSpringConstant=5000
    bAngularSpring=true
    bOverrideAngularDamping=true
    bUseGravityOverride=true
    gravityScale=0.35
    simulationSpace=RootRelative
    alphaScaleBiasClamp {bInterpResult=true, interpSpeedDecreasing=1, interpSpeedIncreasing=5}
    boundBone {boneName=dyn_skirt_sd_l_02}
    componentLinearAccScale=<0.3 0.3 0.3>
    componentLinearVelScale=<0.3 0.3 0.3>
    gravityOverride {z=-980}
    physicsBodyDefinitions [{
      boundBone {boneName=dyn_skirt_sd_l_02}
      boxExtents=<20 30 30>
      constraintSetup {angularLimitsMax=<0 0 5>, angularLimitsMin=<0 -50 -5>, angularTarget=<1 0 0>}
      localJointOffset=<4 0 0>
     }]
    simSpaceSettings {simSpaceAngularAlpha=0.8}
   }
   n45 = Anim Dynamics - Bone: dyn_skirt_sd_l_01(ComponentPose=@n44, Alpha=1, GravityOverride="0,0,-980", ExternalForce="0,0,0") {
    angularDampingOverride=1
    angularSpringConstant=20000
    bAngularSpring=true
    bOverrideAngularDamping=true
    bUseGravityOverride=true
    gravityScale=0.35
    simulationSpace=RootRelative
    alphaScaleBiasClamp {bInterpResult=true, interpSpeedDecreasing=1, interpSpeedIncreasing=5}
    boundBone {boneName=dyn_skirt_sd_l_01}
    componentLinearAccScale=<0.3 0.3 0.3>
    componentLinearVelScale=<0.3 0.3 0.3>
    gravityOverride {z=-980}
    physicsBodyDefinitions [{
      boundBone {boneName=dyn_skirt_sd_l_01}
      boxExtents=<20 30 30>
      constraintSetup {angularLimitsMax=<0 0 10>, angularLimitsMin=<0 -10 -10>, angularTarget=<1 0 0>}
      localJointOffset=<4 0 0>
     }]
    simSpaceSettings {simSpaceAngularAlpha=0.6}
   }
   n46 = Anim Dynamics - Bone: dyn_skirt_fr_r_03(ComponentPose=@n45, Alpha=1, GravityOverride="0,0,-980", ExternalForce="0,0,0") {
    angularDampingOverride=1
    angularSpringConstant=2000
    bAngularSpring=true
    bOverrideAngularDamping=true
    bUseGravityOverride=true
    gravityScale=0.35
    simulationSpace=RootRelative
    boundBone {boneName=dyn_skirt_fr_r_03}
    componentLinearAccScale=<0.3 0.3 0.3>
    componentLinearVelScale=<0.3 0.3 0.3>
    gravityOverride {z=-980}
    physicsBodyDefinitions [{
      boundBone {boneName=dyn_skirt_fr_r_03}
      boxExtents=<20 30 30>
      constraintSetup {angularLimitsMax=<0 0 0>, angularLimitsMin=<0 -30 0>, angularTarget=<1 0 0>}
      localJointOffset=<4 0 0>
     }]
    simSpaceSettings {simSpaceAngularAlpha=0.4}
   }
   n47 = Anim Dynamics - Bone: dyn_skirt_fr_l_01(ComponentPose=@n46, Alpha=1, GravityOverride="0,0,-980", ExternalForce="0,0,0") {
    angularDampingOverride=1
    angularSpringConstant=20000
    bAngularSpring=true
    bOverrideAngularDamping=true
    bUseGravityOverride=true
    gravityScale=0.35
    simulationSpace=RootRelative
    boundBone {boneName=dyn_skirt_fr_l_01}
    componentLinearAccScale=<0.3 0.3 0.3>
    componentLinearVelScale=<0.3 0.3 0.3>
    gravityOverride {z=-980}
    physicsBodyDefinitions [{
      boundBone {boneName=dyn_skirt_fr_l_01}
      boxExtents=<20 30 30>
      constraintSetup {angularLimitsMax=<0 0 10>, angularLimitsMin=<0 -10 -10>, angularTarget=<1 0 0>}
      localJointOffset=<4 0 0>
     }]
    simSpaceSettings {simSpaceAngularAlpha=0.4}
   }
   n48 = Anim Dynamics - Bone: dyn_skirt_bk_l_03(ComponentPose=@n47, Alpha=1, GravityOverride="0,0,-980", ExternalForce="0,0,0") {
    angularDampingOverride=1
    angularSpringConstant=5000
    bAngularSpring=true
    bOverrideAngularDamping=true
    bUseGravityOverride=true
    gravityScale=0.35
    simulationSpace=RootRelative
    alphaScaleBiasClamp {bInterpResult=true, interpSpeedDecreasing=1, interpSpeedIncreasing=5}
    boundBone {boneName=dyn_skirt_bk_l_03}
    componentAppliedLinearAccClamp=<200000 200000 200000>
    componentLinearAccScale=<0.3 0.3 0.3>
    componentLinearVelScale=<0.3 0.3 0.3>
    gravityOverride {z=-980}
    physicsBodyDefinitions [{
      boundBone {boneName=dyn_skirt_bk_l_03}
      boxExtents=<30 20 20>
      constraintSetup {angularLimitsMax=<0 10 10>, angularLimitsMin=<0 -10 -10>, angularTarget=<1 0 0>}
      localJointOffset=<4 0 0>
     }]
    simSpaceSettings {maxAngularAcceleration=30000, maxAngularVelocity=30000, simSpaceAngularAlpha=0.8, externalAngularVelocity=<0.3 0.3 0.3>}
   }
   n49 = Anim Dynamics - Bone: dyn_skirt_bk_l_02(ComponentPose=@n48, Alpha=1, GravityOverride="0,0,-980", ExternalForce="0,0,0") {
    angularDampingOverride=1
    angularSpringConstant=5000
    bAngularSpring=true
    bOverrideAngularDamping=true
    bUseGravityOverride=true
    gravityScale=0.35
    simulationSpace=RootRelative
    alphaScaleBiasClamp {bInterpResult=true, interpSpeedDecreasing=1, interpSpeedIncreasing=5}
    boundBone {boneName=dyn_skirt_bk_l_02}
    componentAppliedLinearAccClamp=<200000 200000 200000>
    componentLinearAccScale=<0.3 0.3 0.3>
    componentLinearVelScale=<0.3 0.3 0.3>
    gravityOverride {z=-980}
    physicsBodyDefinitions [{
      boundBone {boneName=dyn_skirt_bk_l_02}
      boxExtents=<30 20 20>
      constraintSetup {angularLimitsMax=<0 -5 0>, angularLimitsMin=<0 -10 0>, angularTarget=<1 0 0>}
      localJointOffset=<4 0 0>
     }]
    simSpaceSettings {maxAngularAcceleration=30000, maxAngularVelocity=30000, simSpaceAngularAlpha=0.8, externalAngularVelocity=<0.3 0.3 0.3>}
   }
   n50 = Anim Dynamics - Bone: dyn_skirt_bk_l_01(ComponentPose=@n49, Alpha=1, GravityOverride="0,0,-980", ExternalForce="0,0,0") {
    angularDampingOverride=1
    angularSpringConstant=20000
    bAngularSpring=true
    bOverrideAngularDamping=true
    bUseGravityOverride=true
    gravityScale=0.35
    simulationSpace=RootRelative
    alphaScaleBiasClamp {bInterpResult=true, interpSpeedDecreasing=1, interpSpeedIncreasing=5}
    boundBone {boneName=dyn_skirt_bk_l_01}
    componentAppliedLinearAccClamp=<200000 200000 200000>
    componentLinearAccScale=<0.3 0.3 0.3>
    componentLinearVelScale=<0.3 0.3 0.3>
    gravityOverride {z=-980}
    physicsBodyDefinitions [{
      boundBone {boneName=dyn_skirt_bk_l_01}
      boxExtents=<30 20 20>
      constraintSetup {angularLimitsMax=<0 0 10>, angularLimitsMin=<0 -50 -10>, angularTarget=<1 0 0>}
      localJointOffset=<4 0 0>
     }]
    simSpaceSettings {maxAngularAcceleration=30000, maxAngularVelocity=30000, simSpaceAngularAlpha=0.6, externalAngularVelocity=<0.3 0.3 0.3>}
   }
   n51 = Anim Dynamics - Bone: dyn_skirt_bk_c_03(ComponentPose=@n50, Alpha=1, GravityOverride="0,0,-980", ExternalForce="0,0,0") {
    angularDampingOverride=1
    angularSpringConstant=5000
    bAngularSpring=true
    bOverrideAngularDamping=true
    bUseGravityOverride=true
    gravityScale=0.35
    simulationSpace=RootRelative
    alphaScaleBiasClamp {bInterpResult=true, interpSpeedDecreasing=1, interpSpeedIncreasing=5}
    boundBone {boneName=dyn_skirt_bk_c_03}
    componentAppliedLinearAccClamp=<200000 200000 200000>
    componentLinearAccScale=<0.3 0.3 0.3>
    componentLinearVelScale=<0.3 0.3 0.3>
    gravityOverride {z=-980}
    physicsBodyDefinitions [{
      boundBone {boneName=dyn_skirt_bk_c_03}
      boxExtents=<30 20 20>
      constraintSetup {angularLimitsMax=<0 10 10>, angularLimitsMin=<0 -10 -5>, angularTarget=<1 0 0>}
      localJointOffset=<4 0 0>
     }]
    simSpaceSettings {maxAngularAcceleration=30000, maxAngularVelocity=30000, simSpaceAngularAlpha=0.6, externalAngularVelocity=<0.3 0.3 0.3>}
   }
   n52 = Anim Dynamics - Bone: dyn_skirt_bk_c_02(ComponentPose=@n51, Alpha=1, GravityOverride="0,0,-980", ExternalForce="0,0,0") {
    angularDampingOverride=1
    angularSpringConstant=2000
    bAngularSpring=true
    bOverrideAngularDamping=true
    bUseGravityOverride=true
    gravityScale=0.35
    simulationSpace=RootRelative
    alphaScaleBiasClamp {bInterpResult=true, interpSpeedDecreasing=1, interpSpeedIncreasing=5}
    boundBone {boneName=dyn_skirt_bk_c_02}
    componentAppliedLinearAccClamp=<200000 200000 200000>
    componentLinearAccScale=<0.3 0.3 0.3>
    componentLinearVelScale=<0.3 0.3 0.3>
    gravityOverride {z=-980}
    physicsBodyDefinitions [{
      boundBone {boneName=dyn_skirt_bk_c_02}
      boxExtents=<30 20 20>
      constraintSetup {angularLimitsMax=<0 0 0>, angularLimitsMin=<0 -10 0>, angularTarget=<1 0 0>}
      localJointOffset=<4 0 0>
     }]
    simSpaceSettings {maxAngularAcceleration=30000, maxAngularVelocity=30000, simSpaceAngularAlpha=0.6, externalAngularVelocity=<0.3 0.3 0.3>}
   }
   n53 = Anim Dynamics - Bone: dyn_skirt_bk_c_01(ComponentPose=@n52, Alpha=1, GravityOverride="0,0,-980", ExternalForce="0,0,0") {
    angularDampingOverride=1
    angularSpringConstant=5000
    bAngularSpring=true
    bOverrideAngularDamping=true
    bUseGravityOverride=true
    gravityScale=0.35
    simulationSpace=RootRelative
    alphaScaleBiasClamp {bInterpResult=true, interpSpeedDecreasing=1, interpSpeedIncreasing=5}
    boundBone {boneName=dyn_skirt_bk_c_01}
    componentAppliedLinearAccClamp=<200000 200000 200000>
    componentLinearAccScale=<0.3 0.3 0.3>
    componentLinearVelScale=<0.3 0.3 0.3>
    gravityOverride {z=-980}
    physicsBodyDefinitions [{
      boundBone {boneName=dyn_skirt_bk_c_01}
      boxExtents=<30 20 20>
      constraintSetup {angularLimitsMax=<0 0 10>, angularLimitsMin=<0 -80 -2>, angularTarget=<1 0 0>}
      localJointOffset=<4 0 0>
     }]
    simSpaceSettings {maxAngularAcceleration=30000, maxAngularVelocity=30000, simSpaceAngularAlpha=0.6, externalAngularVelocity=<0.3 0.3 0.3>}
   }
   n54 = Anim Dynamics - Bone: dyn_skirt_bk_r_01(ComponentPose=@n53, Alpha=1, GravityOverride="0,0,-980", ExternalForce="0,0,0") {
    angularDampingOverride=1
    angularSpringConstant=20000
    bAngularSpring=true
    bOverrideAngularDamping=true
    bUseGravityOverride=true
    gravityScale=0.35
    simulationSpace=RootRelative
    alphaScaleBiasClamp {bInterpResult=true, interpSpeedDecreasing=1, interpSpeedIncreasing=5}
    boundBone {boneName=dyn_skirt_bk_r_01}
    componentAppliedLinearAccClamp=<200000 200000 200000>
    componentLinearAccScale=<0.3 0.3 0.3>
    componentLinearVelScale=<0.3 0.3 0.3>
    gravityOverride {z=-980}
    physicsBodyDefinitions [{
      boundBone {boneName=dyn_skirt_bk_r_01}
      boxExtents=<30 20 20>
      constraintSetup {angularLimitsMax=<0 0 10>, angularLimitsMin=<0 -50 -10>, angularTarget=<1 0 0>}
      localJointOffset=<4 0 0>
     }]
    simSpaceSettings {maxAngularAcceleration=30000, maxAngularVelocity=30000, simSpaceAngularAlpha=0.6, externalAngularVelocity=<0.3 0.3 0.3>}
   }
   n55 = Anim Dynamics - Bone: dyn_skirt_bk_r_02(ComponentPose=@n54, Alpha=1, GravityOverride="0,0,-980", ExternalForce="0,0,0") {
    angularDampingOverride=1
    angularSpringConstant=5000
    bAngularSpring=true
    bOverrideAngularDamping=true
    bUseGravityOverride=true
    gravityScale=0.35
    simulationSpace=RootRelative
    alphaScaleBiasClamp {bInterpResult=true, interpSpeedDecreasing=1, interpSpeedIncreasing=5}
    boundBone {boneName=dyn_skirt_bk_r_02}
    componentAppliedLinearAccClamp=<200000 200000 200000>
    componentLinearAccScale=<0.3 0.3 0.3>
    componentLinearVelScale=<0.3 0.3 0.3>
    gravityOverride {z=-980}
    physicsBodyDefinitions [{
      boundBone {boneName=dyn_skirt_bk_r_02}
      boxExtents=<30 20 20>
      constraintSetup {angularLimitsMax=<0 5 5>, angularLimitsMin=<0 -10 -5>, angularTarget=<1 0 0>}
      localJointOffset=<4 0 0>
     }]
    simSpaceSettings {maxAngularAcceleration=30000, maxAngularVelocity=30000, simSpaceAngularAlpha=0.8, externalAngularVelocity=<0.3 0.3 0.3>}
   }
   n56 = Anim Dynamics - Bone: dyn_skirt_bk_r_03(ComponentPose=@n55, Alpha=1, GravityOverride="0,0,-980", ExternalForce="0,0,0") {
    angularDampingOverride=1
    angularSpringConstant=5000
    bAngularSpring=true
    bOverrideAngularDamping=true
    bUseGravityOverride=true
    gravityScale=0.35
    simulationSpace=RootRelative
    alphaScaleBiasClamp {bInterpResult=true, interpSpeedDecreasing=1, interpSpeedIncreasing=5}
    boundBone {boneName=dyn_skirt_bk_r_03}
    componentAppliedLinearAccClamp=<200000 200000 200000>
    componentLinearAccScale=<0.3 0.3 0.3>
    componentLinearVelScale=<0.3 0.3 0.3>
    gravityOverride {z=-980}
    physicsBodyDefinitions [{
      boundBone {boneName=dyn_skirt_bk_r_03}
      boxExtents=<30 20 20>
      constraintSetup {angularLimitsMax=<0 10 10>, angularLimitsMin=<0 -10 -10>, angularTarget=<1 0 0>}
      localJointOffset=<4 0 0>
     }]
    simSpaceSettings {maxAngularAcceleration=30000, maxAngularVelocity=30000, simSpaceAngularAlpha=0.8, externalAngularVelocity=<0.3 0.3 0.3>}
   }
   n57 = Blend Poses by bool(
    BlendPose_0=Component To Local(ComponentPose=@n56)
    BlendPose_1=Blend Poses by bool(BlendPose_0=Component To Local(ComponentPose=@n40), BlendPose_1=Component To Local(ComponentPose=@n20)) {blendType=Sinusoidal}
   ) {blendType=Sinusoidal}
   n58 = Anim Dynamics - Bone: dyn_skirt_sd_r_03(ComponentPose=Local To Component(LocalPose=Use cached pose 'start'()), Alpha=1, GravityOverride="0,0,-980", ExternalForce="0,0,0") {
    angularDampingOverride=1
    angularSpringConstant=8000
    bAngularSpring=true
    bOverrideAngularDamping=true
    bUseGravityOverride=true
    gravityScale=0.35
    simulationSpace=BoneRelative
    alphaScaleBiasClamp {bInterpResult=true, interpSpeedDecreasing=1, interpSpeedIncreasing=5}
    boundBone {boneName=dyn_skirt_sd_r_03}
    componentLinearAccScale=<0.3 0.3 0.3>
    componentLinearVelScale=<0.3 0.3 0.3>
    gravityOverride {z=-980}
    physicsBodyDefinitions [{
      boundBone {boneName=dyn_skirt_sd_r_03}
      boxExtents=<30 20 20>
      constraintSetup {angularLimitsMax=<0 10 10>, angularLimitsMin=<0 -10 -10>, angularTarget=<1 0 0>}
      localJointOffset=<4 0 0>
     }]
    relativeSpaceBone {boneName=thigh_r}
    simSpaceSettings {simSpaceAngularAlpha=0.5}
   }
   n59 = Anim Dynamics - Bone: dyn_skirt_sd_r_01(ComponentPose=@n58, Alpha=1, GravityOverride="0,0,-980", ExternalForce="0,0,0") {
    angularDampingOverride=1
    angularSpringConstant=8000
    bAngularSpring=true
    bOverrideAngularDamping=true
    bUseGravityOverride=true
    gravityScale=0.35
    simulationSpace=RootRelative
    alphaScaleBiasClamp {bInterpResult=true, interpSpeedDecreasing=1, interpSpeedIncreasing=5}
    boundBone {boneName=dyn_skirt_sd_r_01}
    componentLinearAccScale=<0.3 0.3 0.3>
    componentLinearVelScale=<0.3 0.3 0.3>
    gravityOverride {z=-980}
    physicsBodyDefinitions [{
      boundBone {boneName=dyn_skirt_sd_r_01}
      boxExtents=<20 30 30>
      constraintSetup {angularLimitsMax=<0 0 10>, angularLimitsMin=<0 -10 -10>, angularTarget=<1 0 0>}
      localJointOffset=<4 0 0>
     }]
    simSpaceSettings {simSpaceAngularAlpha=0.5}
   }
   n60 = Anim Dynamics - Bone: dyn_skirt_sd_l_03(ComponentPose=@n59, Alpha=1, GravityOverride="0,0,-980", ExternalForce="0,0,0") {
    angularDampingOverride=1
    angularSpringConstant=8000
    bAngularSpring=true
    bOverrideAngularDamping=true
    bUseGravityOverride=true
    gravityScale=0.35
    simulationSpace=RootRelative
    alphaScaleBiasClamp {bInterpResult=true, interpSpeedDecreasing=1, interpSpeedIncreasing=5}
    boundBone {boneName=dyn_skirt_sd_l_03}
    componentLinearAccScale=<0.3 0.3 0.3>
    componentLinearVelScale=<0.3 0.3 0.3>
    gravityOverride {z=-980}
    physicsBodyDefinitions [{
      boundBone {boneName=dyn_skirt_sd_l_03}
      boxExtents=<30 20 20>
      constraintSetup {angularLimitsMax=<0 10 10>, angularLimitsMin=<0 -10 -10>, angularTarget=<1 0 0>}
      localJointOffset=<4 0 0>
     }]
    simSpaceSettings {simSpaceAngularAlpha=0.5}
   }
   n61 = Anim Dynamics - Bone: dyn_skirt_sd_l_02(ComponentPose=@n60, Alpha=1, GravityOverride="0,0,-980", ExternalForce="0,0,0") {
    angularDampingOverride=1
    angularSpringConstant=8000
    bAngularSpring=true
    bOverrideAngularDamping=true
    bUseGravityOverride=true
    gravityScale=0.35
    simulationSpace=RootRelative
    alphaScaleBiasClamp {bInterpResult=true, interpSpeedDecreasing=1, interpSpeedIncreasing=5}
    boundBone {boneName=dyn_skirt_sd_l_02}
    componentLinearAccScale=<0.3 0.3 0.3>
    componentLinearVelScale=<0.3 0.3 0.3>
    gravityOverride {z=-980}
    physicsBodyDefinitions [{
      boundBone {boneName=dyn_skirt_sd_l_02}
      boxExtents=<20 30 30>
      constraintSetup {angularLimitsMax=<0 0 5>, angularLimitsMin=<0 -10 -5>, angularTarget=<1 0 0>}
      localJointOffset=<4 0 0>
     }]
    simSpaceSettings {simSpaceAngularAlpha=0.5}
   }
   n62 = Anim Dynamics - Bone: dyn_skirt_sd_l_01(ComponentPose=@n61, Alpha=1, GravityOverride="0,0,-980", ExternalForce="0,0,0") {
    angularDampingOverride=1
    angularSpringConstant=8000
    bAngularSpring=true
    bOverrideAngularDamping=true
    bUseGravityOverride=true
    gravityScale=0.35
    simulationSpace=RootRelative
    alphaScaleBiasClamp {bInterpResult=true, interpSpeedDecreasing=1, interpSpeedIncreasing=5}
    boundBone {boneName=dyn_skirt_sd_l_01}
    componentLinearAccScale=<0.3 0.3 0.3>
    componentLinearVelScale=<0.3 0.3 0.3>
    gravityOverride {z=-980}
    physicsBodyDefinitions [{
      boundBone {boneName=dyn_skirt_sd_l_01}
      boxExtents=<20 30 30>
      constraintSetup {angularLimitsMax=<0 0 10>, angularLimitsMin=<0 -10 -10>, angularTarget=<1 0 0>}
      localJointOffset=<4 0 0>
     }]
    simSpaceSettings {simSpaceAngularAlpha=0.5}
   }
   n63 = Anim Dynamics - Bone: dyn_skirt_fr_r_03(ComponentPose=@n62, Alpha=1, GravityOverride="0,0,-980", ExternalForce="0,0,0") {
    angularDampingOverride=1
    angularSpringConstant=10000
    bAngularSpring=true
    bOverrideAngularDamping=true
    bUseGravityOverride=true
    gravityScale=0.35
    simulationSpace=RootRelative
    alphaScaleBiasClamp {bInterpResult=true, interpSpeedDecreasing=1, interpSpeedIncreasing=5}
    boundBone {boneName=dyn_skirt_fr_r_03}
    componentLinearAccScale=<0.3 0.3 0.3>
    componentLinearVelScale=<0.3 0.3 0.3>
    gravityOverride {z=-980}
    physicsBodyDefinitions [{
      boundBone {boneName=dyn_skirt_fr_r_03}
      boxExtents=<30 20 20>
      constraintSetup {angularLimitsMax=<0 10 0>, angularLimitsMin=<0 -70 0>, angularTarget=<1 0 0>}
      localJointOffset=<4 0 0>
     }]
    simSpaceSettings {simSpaceAngularAlpha=0.6}
   }
   n64 = Anim Dynamics - Bone: dyn_skirt_fr_r_02(ComponentPose=@n63, Alpha=1, GravityOverride="0,0,-980", ExternalForce="0,0,0") {
    angularDampingOverride=1
    angularSpringConstant=5000
    bAngularSpring=true
    bOverrideAngularDamping=true
    bUseGravityOverride=true
    gravityScale=0.35
    simulationSpace=RootRelative
    alphaScaleBiasClamp {bInterpResult=true, interpSpeedDecreasing=1, interpSpeedIncreasing=5}
    boundBone {boneName=dyn_skirt_fr_r_02}
    componentLinearAccScale=<0.3 0.3 0.3>
    componentLinearVelScale=<0.3 0.3 0.3>
    gravityOverride {z=-980}
    physicsBodyDefinitions [{
      boundBone {boneName=dyn_skirt_fr_r_02}
      boxExtents=<20 30 30>
      constraintSetup {angularLimitsMax=<0 0 0>, angularLimitsMin=<0 -30 0>, angularTarget=<1 0 0>}
      localJointOffset=<4 0 0>
     }]
    simSpaceSettings {simSpaceAngularAlpha=0.6}
   }
   n65 = Anim Dynamics - Bone: dyn_skirt_fr_r_01(ComponentPose=@n64, Alpha=1, GravityOverride="0,0,-980", ExternalForce="0,0,0") {
    angularDampingOverride=1
    angularSpringConstant=20000
    bAngularSpring=true
    bOverrideAngularDamping=true
    bUseGravityOverride=true
    gravityScale=0.35
    simulationSpace=RootRelative
    alphaScaleBiasClamp {bInterpResult=true, interpSpeedDecreasing=1, interpSpeedIncreasing=5}
    boundBone {boneName=dyn_skirt_fr_r_01}
    componentLinearAccScale=<0.3 0.3 0.3>
    componentLinearVelScale=<0.3 0.3 0.3>
    gravityOverride {z=-980}
    physicsBodyDefinitions [{
      boundBone {boneName=dyn_skirt_fr_r_01}
      boxExtents=<20 30 30>
      constraintSetup {angularLimitsMax=<0 0 10>, angularLimitsMin=<0 -10 -10>, angularTarget=<1 0 0>}
      localJointOffset=<4 0 0>
     }]
    simSpaceSettings {simSpaceAngularAlpha=0.6}
   }
   n66 = Anim Dynamics - Bone: dyn_skirt_fr_c_03(ComponentPose=@n65, Alpha=1, GravityOverride="0,0,-980", ExternalForce="0,0,0") {
    angularDampingOverride=1
    angularSpringConstant=5000
    bAngularSpring=true
    bOverrideAngularDamping=true
    bUseGravityOverride=true
    gravityScale=0.35
    simulationSpace=BoneRelative
    alphaScaleBiasClamp {bInterpResult=true, interpSpeedDecreasing=1, interpSpeedIncreasing=5}
    boundBone {boneName=dyn_skirt_fr_c_03}
    componentLinearAccScale=<0.3 0.3 0.3>
    componentLinearVelScale=<0.3 0.3 0.3>
    gravityOverride {z=-980}
    physicsBodyDefinitions [{
      boundBone {boneName=dyn_skirt_fr_c_03}
      boxExtents=<20 30 30>
      constraintSetup {angularLimitsMax=<0 0 8>, angularLimitsMin=<0 -5 -8>, angularTarget=<1 0 0>}
      localJointOffset=<4 0 0>
     }]
    relativeSpaceBone {boneName=thigh_r}
    simSpaceSettings {maxAngularAcceleration=5000, maxAngularVelocity=5000, simSpaceAngularAlpha=0.1}
   }
   n67 = Anim Dynamics - Bone: dyn_skirt_fr_l_03(ComponentPose=@n66, Alpha=1, GravityOverride="0,0,-980", ExternalForce="0,0,0") {
    angularDampingOverride=1
    angularSpringConstant=10000
    bAngularSpring=true
    bOverrideAngularDamping=true
    bUseGravityOverride=true
    gravityScale=0.35
    simulationSpace=BoneRelative
    alphaScaleBiasClamp {bInterpResult=true, interpSpeedDecreasing=1, interpSpeedIncreasing=5}
    boundBone {boneName=dyn_skirt_fr_l_03}
    componentLinearAccScale=<0.3 0.3 0.3>
    componentLinearVelScale=<0.3 0.3 0.3>
    gravityOverride {z=-980}
    physicsBodyDefinitions [{
      boundBone {boneName=dyn_skirt_fr_l_03}
      boxExtents=<30 20 20>
      constraintSetup {angularLimitsMax=<0 10 10>, angularLimitsMin=<0 -70 -10>, angularTarget=<1 0 0>}
      localJointOffset=<4 0 0>
     }]
    relativeSpaceBone {boneName=thigh_r}
    simSpaceSettings {simSpaceAngularAlpha=0.6}
   }
   n68 = Anim Dynamics - Bone: dyn_skirt_fr_l_01(ComponentPose=@n67, Alpha=1, GravityOverride="0,0,-980", ExternalForce="0,0,0") {
    angularDampingOverride=1
    angularSpringConstant=20000
    bAngularSpring=true
    bOverrideAngularDamping=true
    bUseGravityOverride=true
    gravityScale=0.35
    simulationSpace=RootRelative
    alphaScaleBiasClamp {bInterpResult=true, interpSpeedDecreasing=1, interpSpeedIncreasing=5}
    boundBone {boneName=dyn_skirt_fr_l_01}
    componentLinearAccScale=<0.3 0.3 0.3>
    componentLinearVelScale=<0.3 0.3 0.3>
    gravityOverride {z=-980}
    physicsBodyDefinitions [{
      boundBone {boneName=dyn_skirt_fr_l_01}
      boxExtents=<20 30 30>
      constraintSetup {angularLimitsMax=<0 0 10>, angularLimitsMin=<0 -10 -10>, angularTarget=<1 0 0>}
      localJointOffset=<4 0 0>
     }]
    simSpaceSettings {simSpaceAngularAlpha=0.6}
   }
   n69 = Anim Dynamics - Bone: dyn_skirt_bk_l_03(ComponentPose=@n68, Alpha=1, GravityOverride="0,0,-980", ExternalForce="0,0,0") {
    angularDampingOverride=1
    angularSpringConstant=8000
    bAngularSpring=true
    bOverrideAngularDamping=true
    bUseGravityOverride=true
    gravityScale=0.35
    simulationSpace=RootRelative
    alphaScaleBiasClamp {bInterpResult=true, interpSpeedDecreasing=1, interpSpeedIncreasing=5}
    boundBone {boneName=dyn_skirt_bk_l_03}
    componentAppliedLinearAccClamp=<200000 200000 200000>
    componentLinearAccScale=<0.3 0.3 0.3>
    componentLinearVelScale=<0.3 0.3 0.3>
    gravityOverride {z=-980}
    physicsBodyDefinitions [{
      boundBone {boneName=dyn_skirt_bk_l_03}
      boxExtents=<30 20 20>
      constraintSetup {angularLimitsMax=<0 10 10>, angularLimitsMin=<0 -10 -10>, angularTarget=<1 0 0>}
      localJointOffset=<4 0 0>
     }]
    simSpaceSettings {maxAngularAcceleration=30000, maxAngularVelocity=30000, simSpaceAngularAlpha=0.5, externalAngularVelocity=<0.3 0.3 0.3>}
   }
   n70 = Anim Dynamics - Bone: dyn_skirt_bk_l_02(ComponentPose=@n69, Alpha=1, GravityOverride="0,0,-980", ExternalForce="0,0,0") {
    angularDampingOverride=1
    angularSpringConstant=8000
    bAngularSpring=true
    bOverrideAngularDamping=true
    bUseGravityOverride=true
    gravityScale=0.35
    simulationSpace=RootRelative
    alphaScaleBiasClamp {bInterpResult=true, interpSpeedDecreasing=1, interpSpeedIncreasing=5}
    boundBone {boneName=dyn_skirt_bk_l_02}
    componentAppliedLinearAccClamp=<200000 200000 200000>
    componentLinearAccScale=<0.3 0.3 0.3>
    componentLinearVelScale=<0.3 0.3 0.3>
    gravityOverride {z=-980}
    physicsBodyDefinitions [{
      boundBone {boneName=dyn_skirt_bk_l_02}
      boxExtents=<30 20 20>
      constraintSetup {angularLimitsMax=<0 0 0>, angularLimitsMin=<0 -10 0>, angularTarget=<1 0 0>}
      localJointOffset=<4 0 0>
     }]
    simSpaceSettings {maxAngularAcceleration=30000, maxAngularVelocity=30000, simSpaceAngularAlpha=0.5, externalAngularVelocity=<0.3 0.3 0.3>}
   }
   n71 = Anim Dynamics - Bone: dyn_skirt_bk_l_01(ComponentPose=@n70, Alpha=1, GravityOverride="0,0,-980", ExternalForce="0,0,0") {
    angularDampingOverride=1
    angularSpringConstant=8000
    bAngularSpring=true
    bOverrideAngularDamping=true
    bUseGravityOverride=true
    gravityScale=0.35
    simulationSpace=RootRelative
    alphaScaleBiasClamp {bInterpResult=true, interpSpeedDecreasing=1, interpSpeedIncreasing=5}
    boundBone {boneName=dyn_skirt_bk_l_01}
    componentAppliedLinearAccClamp=<200000 200000 200000>
    componentLinearAccScale=<0.3 0.3 0.3>
    componentLinearVelScale=<0.3 0.3 0.3>
    gravityOverride {z=-980}
    physicsBodyDefinitions [{
      boundBone {boneName=dyn_skirt_bk_l_01}
      boxExtents=<30 20 20>
      constraintSetup {angularLimitsMax=<0 0 10>, angularLimitsMin=<0 -10 -10>, angularTarget=<1 0 0>}
      localJointOffset=<5 0 0>
     }]
    simSpaceSettings {maxAngularAcceleration=30000, maxAngularVelocity=30000, simSpaceAngularAlpha=0.5, externalAngularVelocity=<0.3 0.3 0.3>}
   }
   n72 = Anim Dynamics - Bone: dyn_skirt_bk_c_03(ComponentPose=@n71, Alpha=1, GravityOverride="0,0,-980", ExternalForce="0,0,0") {
    angularDampingOverride=1
    angularSpringConstant=8000
    bAngularSpring=true
    bOverrideAngularDamping=true
    bUseGravityOverride=true
    gravityScale=0.35
    simulationSpace=RootRelative
    alphaScaleBiasClamp {bInterpResult=true, interpSpeedDecreasing=1, interpSpeedIncreasing=5}
    boundBone {boneName=dyn_skirt_bk_c_03}
    componentAppliedLinearAccClamp=<200000 200000 200000>
    componentLinearAccScale=<0.3 0.3 0.3>
    componentLinearVelScale=<0.3 0.3 0.3>
    gravityOverride {z=-980}
    physicsBodyDefinitions [{
      boundBone {boneName=dyn_skirt_bk_c_03}
      boxExtents=<30 20 20>
      constraintSetup {angularLimitsMax=<0 10 10>, angularLimitsMin=<0 -10 -5>, angularTarget=<1 0 0>}
      localJointOffset=<4 0 0>
     }]
    simSpaceSettings {maxAngularAcceleration=30000, maxAngularVelocity=30000, simSpaceAngularAlpha=0.5, externalAngularVelocity=<0.3 0.3 0.3>}
   }
   n73 = Anim Dynamics - Bone: dyn_skirt_bk_c_02(ComponentPose=@n72, Alpha=1, GravityOverride="0,0,-980", ExternalForce="0,0,0") {
    angularDampingOverride=1
    angularSpringConstant=8000
    bAngularSpring=true
    bOverrideAngularDamping=true
    bUseGravityOverride=true
    gravityScale=0.35
    simulationSpace=RootRelative
    alphaScaleBiasClamp {bInterpResult=true, interpSpeedDecreasing=1, interpSpeedIncreasing=5}
    boundBone {boneName=dyn_skirt_bk_c_02}
    componentAppliedLinearAccClamp=<200000 200000 200000>
    componentLinearAccScale=<0.3 0.3 0.3>
    componentLinearVelScale=<0.3 0.3 0.3>
    gravityOverride {z=-980}
    physicsBodyDefinitions [{
      boundBone {boneName=dyn_skirt_bk_c_02}
      boxExtents=<30 20 20>
      constraintSetup {angularLimitsMax=<0 0 0>, angularLimitsMin=<0 -10 0>, angularTarget=<1 0 0>}
      localJointOffset=<4 0 0>
     }]
    simSpaceSettings {maxAngularAcceleration=30000, maxAngularVelocity=30000, simSpaceAngularAlpha=0.5, externalAngularVelocity=<0.3 0.3 0.3>}
   }
   n74 = Anim Dynamics - Bone: dyn_skirt_bk_c_01(ComponentPose=@n73, Alpha=1, GravityOverride="0,0,-980", ExternalForce="0,0,0") {
    angularDampingOverride=1
    angularSpringConstant=8000
    bAngularSpring=true
    bOverrideAngularDamping=true
    bUseGravityOverride=true
    gravityScale=0.35
    simulationSpace=RootRelative
    alphaScaleBiasClamp {bInterpResult=true, interpSpeedDecreasing=1, interpSpeedIncreasing=5}
    boundBone {boneName=dyn_skirt_bk_c_01}
    componentAppliedLinearAccClamp=<200000 200000 200000>
    componentLinearAccScale=<0.3 0.3 0.3>
    componentLinearVelScale=<0.3 0.3 0.3>
    gravityOverride {z=-980}
    physicsBodyDefinitions [{
      boundBone {boneName=dyn_skirt_bk_c_01}
      boxExtents=<30 20 20>
      constraintSetup {angularLimitsMax=<0 0 10>, angularLimitsMin=<0 -10 -2>, angularTarget=<1 0 0>}
      localJointOffset=<5 0 0>
     }]
    simSpaceSettings {maxAngularAcceleration=30000, maxAngularVelocity=30000, simSpaceAngularAlpha=0.5, externalAngularVelocity=<0.3 0.3 0.3>}
   }
   n75 = Anim Dynamics - Bone: dyn_skirt_bk_r_01(ComponentPose=@n74, Alpha=1, GravityOverride="0,0,-980", ExternalForce="0,0,0") {
    angularDampingOverride=1
    angularSpringConstant=8000
    bAngularSpring=true
    bOverrideAngularDamping=true
    bUseGravityOverride=true
    gravityScale=0.35
    simulationSpace=RootRelative
    alphaScaleBiasClamp {bInterpResult=true, interpSpeedDecreasing=1, interpSpeedIncreasing=5}
    boundBone {boneName=dyn_skirt_bk_r_01}
    componentAppliedLinearAccClamp=<200000 200000 200000>
    componentLinearAccScale=<0.3 0.3 0.3>
    componentLinearVelScale=<0.3 0.3 0.3>
    gravityOverride {z=-980}
    physicsBodyDefinitions [{
      boundBone {boneName=dyn_skirt_bk_r_01}
      boxExtents=<30 20 20>
      constraintSetup {angularLimitsMax=<0 0 10>, angularLimitsMin=<0 -10 -10>, angularTarget=<1 0 0>}
      localJointOffset=<5 0 0>
     }]
    simSpaceSettings {maxAngularAcceleration=30000, maxAngularVelocity=30000, simSpaceAngularAlpha=0.5, externalAngularVelocity=<0.3 0.3 0.3>}
   }
   n76 = Anim Dynamics - Bone: dyn_skirt_bk_r_02(ComponentPose=@n75, Alpha=1, GravityOverride="0,0,-980", ExternalForce="0,0,0") {
    angularDampingOverride=1
    angularSpringConstant=8000
    bAngularSpring=true
    bOverrideAngularDamping=true
    bUseGravityOverride=true
    gravityScale=0.35
    simulationSpace=RootRelative
    alphaScaleBiasClamp {bInterpResult=true, interpSpeedDecreasing=1, interpSpeedIncreasing=5}
    boundBone {boneName=dyn_skirt_bk_r_02}
    componentAppliedLinearAccClamp=<200000 200000 200000>
    componentLinearAccScale=<0.3 0.3 0.3>
    componentLinearVelScale=<0.3 0.3 0.3>
    gravityOverride {z=-980}
    physicsBodyDefinitions [{
      boundBone {boneName=dyn_skirt_bk_r_02}
      boxExtents=<30 20 20>
      constraintSetup {angularLimitsMax=<0 0 5>, angularLimitsMin=<0 -10 -5>, angularTarget=<1 0 0>}
      localJointOffset=<4 0 0>
     }]
    simSpaceSettings {maxAngularAcceleration=30000, maxAngularVelocity=30000, simSpaceAngularAlpha=0.5, externalAngularVelocity=<0.3 0.3 0.3>}
   }
   n77 = Anim Dynamics - Bone: dyn_skirt_bk_r_03(ComponentPose=@n76, Alpha=1, GravityOverride="0,0,-980", ExternalForce="0,0,0") {
    angularDampingOverride=1
    angularSpringConstant=8000
    bAngularSpring=true
    bOverrideAngularDamping=true
    bUseGravityOverride=true
    gravityScale=0.35
    simulationSpace=RootRelative
    alphaScaleBiasClamp {bInterpResult=true, interpSpeedDecreasing=1, interpSpeedIncreasing=5}
    boundBone {boneName=dyn_skirt_bk_r_03}
    componentAppliedLinearAccClamp=<200000 200000 200000>
    componentLinearAccScale=<0.3 0.3 0.3>
    componentLinearVelScale=<0.3 0.3 0.3>
    gravityOverride {z=-980}
    physicsBodyDefinitions [{
      boundBone {boneName=dyn_skirt_bk_r_03}
      boxExtents=<30 20 20>
      constraintSetup {angularLimitsMax=<0 10 10>, angularLimitsMin=<0 -10 -10>, angularTarget=<1 0 0>}
      localJointOffset=<4 0 0>
     }]
    simSpaceSettings {maxAngularAcceleration=30000, maxAngularVelocity=30000, simSpaceAngularAlpha=0.5, externalAngularVelocity=<0.3 0.3 0.3>}
   }
   n78 = Component To Local(ComponentPose=@n77)
   n79 = "Transform (Modify) Bone - Bone: base_skirt_bk_c"(
    ComponentPose=Local To Component(LocalPose=Blend Poses by bool(BlendPose_0=@n78, BlendPose_1=@n57) {blendType=Sinusoidal})
    Alpha=1
    Translation="0,0,0"
    Rotation="25,0,0"
    Scale="1,1,1"
   ) {
    rotationMode=BMM_Additive
    rotationSpace=BCS_BoneSpace
    scaleMode=BMM_Additive
    scaleSpace=BCS_BoneSpace
    alphaScaleBiasClamp {bInterpResult=true, interpSpeedDecreasing=4, interpSpeedIncreasing=5}
    boneToModify {boneName=base_skirt_bk_c}
    rotation {pitch=25}
   }
   n80 = "Transform (Modify) Bone - Bone: base_skirt_bk_l"(ComponentPose=@n79, Alpha=1, Translation="0,0,0", Rotation="0,25,0", Scale="1,1,1") {
    rotationMode=BMM_Additive
    rotationSpace=BCS_BoneSpace
    scaleMode=BMM_Additive
    scaleSpace=BCS_BoneSpace
    alphaScaleBiasClamp {bInterpResult=true, interpSpeedDecreasing=4, interpSpeedIncreasing=5}
    boneToModify {boneName=base_skirt_bk_l}
    rotation {yaw=25}
   }
   n81 = "Transform (Modify) Bone - Bone: base_skirt_bk_r"(ComponentPose=@n80, Alpha=1, Translation="0,0,0", Rotation="0,25,0", Scale="1,1,1") {
    rotationMode=BMM_Additive
    rotationSpace=BCS_BoneSpace
    scaleMode=BMM_Additive
    scaleSpace=BCS_BoneSpace
    alphaScaleBiasClamp {bInterpResult=true, interpSpeedDecreasing=4, interpSpeedIncreasing=5}
    boneToModify {boneName=base_skirt_bk_r}
    rotation {yaw=25}
   }
   n82 = "Transform (Modify) Bone - Bone: base_skirt_fr_c"(ComponentPose=@n81, Alpha=1, Translation="0,0,0", Rotation="0,-40,-10", Scale="1,1,1") {
    rotationMode=BMM_Additive
    rotationSpace=BCS_BoneSpace
    alphaScaleBiasClamp {bInterpResult=true, interpSpeedDecreasing=2, interpSpeedIncreasing=5}
    boneToModify {boneName=base_skirt_fr_c}
    rotation {roll=-10, yaw=-40}
   }
   n83 = "Transform (Modify) Bone - Bone: dyn_skirt_fr_r_02"(ComponentPose=@n82, Alpha=1, Translation="0,0,0", Rotation="10,0,0", Scale="1,1,1") {
    rotationMode=BMM_Additive
    rotationSpace=BCS_BoneSpace
    alphaScaleBiasClamp {bInterpResult=true}
    boneToModify {boneName=dyn_skirt_fr_r_02}
    rotation {pitch=10}
   }
   n84 = "Transform (Modify) Bone - Bone: dyn_skirt_fr_l_02"(ComponentPose=@n83, Alpha=1, Translation="0,0,0", Rotation="10,10,0", Scale="1,1,1") {
    rotationMode=BMM_Additive
    rotationSpace=BCS_BoneSpace
    alphaScaleBiasClamp {bInterpResult=true}
    boneToModify {boneName=dyn_skirt_fr_l_02}
    rotation {pitch=10, yaw=10}
   }
   n85 = "Transform (Modify) Bone - Bone: dyn_skirt_fr_c_02"(ComponentPose=@n84, Alpha=1, Translation="0,0,0", Rotation="15,0,0", Scale="1,1,1") {
    rotationMode=BMM_Additive
    rotationSpace=BCS_BoneSpace
    translationMode=BMM_Additive
    translationSpace=BCS_BoneSpace
    alphaScaleBiasClamp {bInterpResult=true}
    boneToModify {boneName=dyn_skirt_fr_c_02}
    rotation {pitch=15}
   }
   n86 = "Transform (Modify) Bone - Bone: base_waist"(ComponentPose=@n85, Alpha=1, Translation="4,0,0", Rotation="0,0,0", Scale="1,1,1") {
    translationMode=BMM_Additive
    translationSpace=BCS_BoneSpace
    alphaScaleBiasClamp {bInterpResult=true, interpSpeedDecreasing=5}
    boneToModify {boneName=base_waist}
    translation {x=4}
   }
   Output Pose(Result=Component To Local(ComponentPose=@n86)) {layerGroup=None}
   n87 = Control Rig!(
    Source=Copy Pose From Mesh() {bUseAttachedParent=true}
    Alpha
    CrouchingMoving2D:double=Get iscrouchwalking() -> (iscrouchwalking:double)
    Crouching:double=Get iscrouching() -> (iscrouching:double)
    targeting:double=Get Targeting() -> (Targeting:double)
    Sprinting:double=Get Sprinting() -> (Sprinting:double)
    slide:double=Get is sliding() -> (is sliding:double)
    bbslide:double=Get bb_slide() -> (bb_slide:double)
    FE:double=Get FE() -> (FE:double)
    Melee:double=Get Melee() -> (Melee:double)
    Target_down:double=Get TagertDown() -> (TagertDown:double)
    Dive:double=Get Diving() -> (Diving:double)
    DBNO:double=Get DBNO() -> (DBNO:double)
    DBNO_caried:double=Get carried()
    FE_Emote:double=Get FE_Emote() -> (FE_Emote:double)
   ) -> (Pose) {controlRigAssetReference {blueprintRigClass=&/F_MED_LumpyMuddy_CtrlRig.F_MED_LumpyMuddy_CtrlRig_C}}
   Save cached pose 'start'(Pose=@n87) {cacheName=start}
  ]}
 EventGraph {nodes [
   Event BlueprintUpdateAnimation()
   > n1 = Sequence!(execute) -> (then_0, then_1, then_2:Exec)
   > n2 = "Gravity Override State Function (Parameter Asset) (Nativized)"(
    CharacterPartAnimInstance=Self-Reference()
    GameTimeInSeconds=GetGameTimeInSeconds()
    DebugJointName=spine_05
    Parameters=&/F_Med_LumpyMuddy_Parameters
   )
   > Set Gravity Override(Gravity Override=@n2)
   n3 = OR Boolean(A=Get bIsTacticalSprinting(), B=Get bIsPlayingEmote())
   Set no emote(execute=@n1.then_1, no emote:double="To Float (Boolean)"(InBool=NOT Boolean(A=@n3))) -> (Output_Get:double)
   n4 = NOT Boolean(A=Get bIsMoving2D())
   n5 = Get bIsCrouching()
   > Set iscrouching(iscrouching:double="To Float (Boolean)"(InBool=AND Boolean(A=@n5, B=@n4))) -> (Output_Get:double)
   > Set is sliding(is sliding:double="To Float (Boolean)"(InBool=Get bIsKneeSliding())) -> (Output_Get:double)
   > Set bb_slide(bb_slide:double="To Float (Boolean)"(InBool=Get bIsBaseballSliding())) -> (Output_Get:double)
   n6 = Get bIsCrouching()
   n7 = Get bIsMoving2D()
   > Set iscrouchwalking(iscrouchwalking:double="To Float (Boolean)"(InBool=AND Boolean(A=@n6, B=@n7))) -> (Output_Get:double)
   n8 = NOT Boolean(A=OR Boolean(A=Get bIsCrouching(), B=Get bIsJumping()))
   n9 = Get bIsTargeting()
   > Set TagertDown(TagertDown:double="To Float (Boolean)"(InBool=AND Boolean(A=@n9, B=@n8))) -> (Output_Get:double)
   n10 = AND Boolean(A=Get bIsJumping(), B=Get bIsSkyDiving())
   n11 = Get bIsSkydiveDiveMode()
   > Set Diving(Diving:double="To Float (Boolean)"(InBool=OR Boolean(A=@n10, B=@n11))) -> (Output_Get:double)
   n12 = Get bIsCrouching()
   n13 = Get bIsMoving2D()
   > Set Crouchwalk(Crouchwalk:double="To Float (Boolean)"(InBool=AND Boolean(A=@n12, B=@n13))) -> (Output_Get:double)
   n14 = NOT Boolean(A=Get bIsMoving2D())
   n15 = Get bIsFalling()
   > Set isjump(isjump:double="To Float (Boolean)"(InBool=AND Boolean(A=@n15, B=@n14))) -> (Output_Get:double)
   n16 = Get bIsJumping()
   n17 = Get bIsCrouching()
   n18 = AND Boolean!(A=Get bIsTargeting(), B=Get bIsMoving2D(), C:bool=NOT Boolean(A=OR Boolean(A=@n16, B=@n17))) -> (ReturnValue)
   > Set Targeting(Targeting:double="To Float (Boolean)"(InBool=NOT Boolean(A=@n18))) -> (Output_Get:double)
   > Set Sprinting(Sprinting:double="To Float (Boolean)"(InBool=Get bIsTacticalSprinting())) -> (Output_Get:double)
   n19 = NOT Boolean(A=OR Boolean(A=Get bIsCrouching(), B=Get bIsTargeting()))
   n20 = OR Boolean!(A=Get bIsMoving2D(), B=Get bIsPlayingEmote(), C:bool=Get bIsJumping()) -> (ReturnValue)
   > Set Moving(Moving:double="To Float (Boolean)"(InBool=AND Boolean!(A=@n20, B=@n19) -> (ReturnValue))) -> (Output_Get:double)
   n21 = Get bIsJumping()
   n22 = Get bIsMoving2D()
   > Set UP(UP:double="To Float (Boolean)"(InBool=AND Boolean(A=@n21, B=@n22))) -> (Output_Get:double)
   n23 = NOT Boolean(A=Get bIsDBNOCarried())
   n24 = Get bIsDBNO()
   > Set DBNO(DBNO:double="To Float (Boolean)"(InBool=AND Boolean(A=@n24, B=@n23))) -> (Output_Get:double)
   n25 = OR Boolean(A=Get bIsTacticalSprinting(), B=Get bIsPlayingEmote(), C=Get bIsSkyDiving())
   > Set no emote_+(no emote_+:double="To Float (Boolean)"(InBool=NOT Boolean(A=@n25))) -> (Output_Get:double)
   n26 = NOT Boolean(A=Get bIsPlayingEmote())
   n27 = NOT Boolean(A=Get bIsPreviewingItemInFrontEnd() -> (bIsPreviewingItemInFrontEnd:bool))
   n28 = Get bIsInFrontEnd()
   > Set FE(FE:double="To Float (Boolean)"(InBool=AND Boolean!(A=@n28, B=@n27, C:bool=@n26) -> (ReturnValue))) -> (Output_Get:double)
   > Set carried(carried:double="To Float (Boolean)"(InBool=Get bIsDBNOCarried())) -> (Output_Get:double)
   n29 = OR Boolean!(A=Get bIsInFrontEnd(), B=Get bIsPlayingEmote()) -> (ReturnValue)
   > Set FE_Emote(FE_Emote:double="To Float (Boolean)"(InBool=NOT Boolean(A=@n29))) -> (Output_Get:double)
   n30 = Cast To FortPlayerAnimInstance(execute=@n1.then_2, Object=GetAnimInstance(self=GetOwnerSkeletalMeshComponent()))
   > Set Melee(Melee:double=0) -> (Output_Get:double)
   > n31 = Switch on EFortWeaponCoreAnimation!(execute, Selection:EFortWeaponCoreAnimation Enum=Get WeaponCoreAnim(self=@n30)) -> (
    Melee:Exec
    Pistol:Exec
    Shotgun:Exec
    PaperBlueprint:Exec
    Rifle:Exec
    MeleeOneHand:Exec
    MachinePistol:Exec
    RocketLauncher:Exec
    GrenadeLauncher:Exec
    GoingCommando:Exec
    AssaultRifle:Exec
    TacticalShotgun:Exec
    SniperRifle:Exec
    TrapPlacement:Exec
    ShoulderLauncher:Exec
    AbilityDecoTool:Exec
    Crossbow:Exec
    C4:Exec
    RemoteControl:Exec
    DualWield:Exec
    AR_BullPup:Exec
    AR_ForwardGrip:Exec
    MedPackPaddles:Exec
    SMG_P90:Exec
    AR_DrumGun:Exec
    Consumable_Small:Exec
    Consumable_Large:Exec
    Balloon:Exec
    MountedTurret:Exec
    CreativeTool:Exec
    ExplosiveBow:Exec
    AshtonIndigo:Exec
    AshtonChicago:Exec
    MeleeDualWield:Exec
    Unarmed:Exec
   )
   n32 = Get bIsTacticalSprinting()
   n33 = Get bIsJumping()
   n34 = Get bIsSkyDiving()
   n35 = Branch(
    execute=[@n31.Consumable_Large, @n31.Consumable_Small, @n31.MeleeDualWield, @n31.MeleeOneHand, @n31.Unarmed]
    Condition=NOT Boolean(A=OR Boolean!(A=@n34, B=@n33, C:bool=@n32) -> (ReturnValue))
   )
   > Set Melee(Melee=1)
   Set Melee(execute=@n35.else, Melee=0)
  ]}
]
variables [
 ar_firingPosition {type=integer}
 bb_slide {type=number}
 carried {type=number}
 Crouchwalk {type=number}
 DBNO {type=number}
 Diving {type=number}
 FE {type=number}
 FE_Emote {type=number}
 Gravity Override {type=Vector}
 is sliding {type=number}
 iscrouching {type=number}
 iscrouchwalking {type=number}
 isjump {type=number}
 Melee {type=number}
 Moving {type=number}
 no emote {type=number}
 no emote_+ {type=number}
 Sprinting {type=number}
 TagertDown {type=number}
 Targeting {type=number}
 UP {type=number}
]
