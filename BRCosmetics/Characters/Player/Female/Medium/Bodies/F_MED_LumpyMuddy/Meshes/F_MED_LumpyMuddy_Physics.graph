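# Physics asset for the skirt of the F_MED_LumpyMuddy female player cosmetic body, covering only skirt, pelvis and thigh bones.
# Eight skirt chains around the front, back and sides each run from a tiny base capsule through three dynamic capsules.
# Chain joints have tight swing limits, with zero twist at the first link, while bases and thighs join the pelvis with free rotation.
# Pelvis and thigh capsules provide body volumes alongside the skirt.
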PhysicsAsset /BRCosmetics/Characters/Player/Female/Medium/Bodies/F_MED_LumpyMuddy/Meshes/F_MED_LumpyMuddy_Physics
bodies {
 base_skirt_bk_c [{shapeType=Capsule, length=0.505, radius=0.505, shapeName=base_skirt_bk_c_capsule, rotation=<rot 0 -90 0>}]
 base_skirt_bk_l [{shapeType=Capsule, length=0.505, radius=0.505, shapeName=base_skirt_bk_l_capsule, rotation=<rot 0 -90 0>}]
 base_skirt_bk_r [{shapeType=Capsule, length=0.505, radius=0.505, shapeName=base_skirt_bk_r_capsule, rotation=<rot 0 -90 0>}]
 base_skirt_fr_c [{shapeType=Capsule, length=0.505, radius=0.505, shapeName=base_skirt_fr_c_capsule, rotation=<rot 0 -90 0>}]
 base_skirt_fr_l [{shapeType=Capsule, length=0.505, radius=0.505, shapeName=base_skirt_fr_l_capsule, rotation=<rot 0 -90 0>}]
 base_skirt_fr_r [{shapeType=Capsule, length=0.505, radius=0.505, shapeName=base_skirt_fr_r_capsule, rotation=<rot 0 -90 0>}]
 base_skirt_sd_l [{shapeType=Capsule, length=0.505, radius=0.505, shapeName=base_skirt_sd_l_capsule, rotation=<rot 0 -90 0>}]
 base_skirt_sd_r [{shapeType=Capsule, length=0.505, radius=0.505, shapeName=base_skirt_sd_r_capsule, rotation=<rot 0 -90 0>}]
 dyn_skirt_bk_c_01 [{shapeType=Capsule, length=5, radius=5, shapeName=dyn_skirt_bk_c_01_capsule, center=<8 0 0>, rotation=<rot 90 0 90>}]
 dyn_skirt_bk_c_02 [{shapeType=Capsule, length=10, radius=5, shapeName=dyn_skirt_bk_c_02_capsule, center=<5 0 0>, rotation=<rot 90 0 90>}]
 dyn_skirt_bk_c_03 [{shapeType=Capsule, length=13.5, radius=5, shapeName=dyn_skirt_bk_c_03_capsule, center=<7 0 0>, rotation=<rot 90 0 90>}]
 dyn_skirt_bk_l_01 [{shapeType=Capsule, length=5, radius=5, shapeName=dyn_skirt_bk_l_01_capsule, center=<8 0 0>, rotation=<rot 90 0 90>}]
 dyn_skirt_bk_l_02 [{shapeType=Capsule, length=9, radius=5, shapeName=dyn_skirt_bk_l_02_capsule, center=<5 0 0>, rotation=<rot 90 0 90>}]
 dyn_skirt_bk_l_03 [{shapeType=Capsule, length=14.5, radius=5, shapeName=dyn_skirt_bk_l_03_capsule, center=<7 0 0>, rotation=<rot 90 0 90>}]
 dyn_skirt_bk_r_01 [{shapeType=Capsule, length=5, radius=5, shapeName=dyn_skirt_bk_r_01_capsule, center=<8 0 0>, rotation=<rot 90 0 90>}]
 dyn_skirt_bk_r_02 [{shapeType=Capsule, length=9.5, radius=5, shapeName=dyn_skirt_bk_r_02_capsule, center=<5 0 0>, rotation=<rot 90 0 90>}]
 dyn_skirt_bk_r_03 [{shapeType=Capsule, length=15.5, radius=5, shapeName=dyn_skirt_bk_r_03_capsule, center=<7 0 0>, rotation=<rot 90 0 90>}]
 dyn_skirt_fr_c_01 [{shapeType=Capsule, length=5, radius=5, shapeName=dyn_skirt_fr_c_01_capsule, center=<8 0 0>, rotation=<rot 90 0 90>}]
 dyn_skirt_fr_c_02 [{shapeType=Capsule, length=10.5, radius=5, shapeName=dyn_skirt_fr_c_02_capsule, center=<5 0 0>, rotation=<rot 90 0 90>}]
 dyn_skirt_fr_c_03 [{shapeType=Capsule, length=14.5, radius=5, shapeName=dyn_skirt_fr_c_03_capsule, center=<7 0 0>, rotation=<rot 90 0 90>}]
 dyn_skirt_fr_l_01 [{shapeType=Capsule, length=5, radius=5, shapeName=dyn_skirt_fr_l_01_capsule, center=<8 0 0>, rotation=<rot 90 0 90>}]
 dyn_skirt_fr_l_02 [{shapeType=Capsule, length=10, radius=5, shapeName=dyn_skirt_fr_l_02_capsule, center=<5 0 0>, rotation=<rot 90 0 90>}]
 dyn_skirt_fr_l_03 [{shapeType=Capsule, length=17.5, radius=5, shapeName=dyn_skirt_fr_l_03_capsule, center=<7 0 0>, rotation=<rot 90 0 90>}]
 dyn_skirt_fr_r_01 [{shapeType=Capsule, length=5, radius=5, shapeName=dyn_skirt_fr_r_01_capsule, center=<8 0 0>, rotation=<rot 90 0 90>}]
 dyn_skirt_fr_r_02 [{shapeType=Capsule, length=11.5, radius=5, shapeName=dyn_skirt_fr_r_02_capsule, center=<5 0 0>, rotation=<rot 90 0 90>}]
 dyn_skirt_fr_r_03 [{shapeType=Capsule, length=17, radius=5, shapeName=dyn_skirt_fr_r_03_capsule, center=<7 0 0>, rotation=<rot 90 0 90>}]
 dyn_skirt_sd_l_01 [{shapeType=Capsule, length=5, radius=5, shapeName=dyn_skirt_sd_l_01_capsule, center=<8 0 0>, rotation=<rot 90 0 90>}]
 dyn_skirt_sd_l_02 [{shapeType=Capsule, length=9.5, radius=5, shapeName=dyn_skirt_sd_l_02_capsule, center=<5 0 0>, rotation=<rot 90 0 90>}]
 dyn_skirt_sd_l_03 [{shapeType=Capsule, length=14.5, radius=5, shapeName=dyn_skirt_sd_l_03_capsule, center=<7 0 0>, rotation=<rot 90 0 90>}]
 dyn_skirt_sd_r_01 [{shapeType=Capsule, length=5, radius=5, shapeName=dyn_skirt_sd_r_01_capsule, center=<8 0 0>, rotation=<rot 90 0 90>}]
 dyn_skirt_sd_r_02 [{shapeType=Capsule, length=8, radius=5, shapeName=dyn_skirt_sd_r_02_capsule, center=<5 0 0>, rotation=<rot 90 0 90>}]
 dyn_skirt_sd_r_03 [{shapeType=Capsule, length=15, radius=5, shapeName=dyn_skirt_sd_r_03_capsule, center=<7 0 0>, rotation=<rot 90 0 90>}]
 pelvis [{shapeType=Capsule, length=13.5, radius=10, shapeName=pelvis_capsule, center=<0 -1.0994 0>, rotation=<rot 90 0 0>}]
 thigh_l [{
   shapeType=Capsule
   length=22.5146
   radius=8.3779
   shapeName=thigh_l_capsule
   center=<-19.4581 -0.2295 0.4977>
   rotation=<rot 0 -99.1939 -86.9029>
  }]
 thigh_r [{
   shapeType=Capsule
   length=22.2439
   radius=8.2887
   shapeName=thigh_r_capsule
   center=<19.7983 0.3412 -0.3635>
   rotation=<rot 0 -94.7753 -83.6097>
  }]
}
constraints [
 {bone1Name=dyn_skirt_bk_c_01, bone2Name=base_skirt_bk_c, swing1LimitDegrees=10, swing2LimitDegrees=10, twistLimitDegrees=0}
 {bone1Name=dyn_skirt_bk_c_02, bone2Name=dyn_skirt_bk_c_01, swing1LimitDegrees=10, swing2LimitDegrees=20, twistLimitDegrees=20}
 {bone1Name=dyn_skirt_bk_c_03, bone2Name=dyn_skirt_bk_c_02, swing1LimitDegrees=10, swing2LimitDegrees=20, twistLimitDegrees=20}
 {bone1Name=dyn_skirt_bk_l_01, bone2Name=base_skirt_bk_l, swing1LimitDegrees=10, swing2LimitDegrees=10, twistLimitDegrees=0}
 {bone1Name=dyn_skirt_bk_l_02, bone2Name=dyn_skirt_bk_l_01, swing1LimitDegrees=10, swing2LimitDegrees=20, twistLimitDegrees=20}
 {bone1Name=dyn_skirt_bk_l_03, bone2Name=dyn_skirt_bk_l_02, swing1LimitDegrees=10, swing2LimitDegrees=20, twistLimitDegrees=20}
 {bone1Name=dyn_skirt_bk_r_01, bone2Name=base_skirt_bk_r, swing1LimitDegrees=10, swing2LimitDegrees=10, twistLimitDegrees=0}
 {bone1Name=dyn_skirt_bk_r_02, bone2Name=dyn_skirt_bk_r_01, swing1LimitDegrees=10, swing2LimitDegrees=20, twistLimitDegrees=20}
 {bone1Name=dyn_skirt_bk_r_03, bone2Name=dyn_skirt_bk_r_02, swing1LimitDegrees=10, swing2LimitDegrees=20, twistLimitDegrees=20}
 {bone1Name=dyn_skirt_fr_c_01, bone2Name=base_skirt_fr_c, swing1LimitDegrees=10, swing2LimitDegrees=10, twistLimitDegrees=0}
 {bone1Name=dyn_skirt_fr_c_02, bone2Name=dyn_skirt_fr_c_01, swing1LimitDegrees=10, swing2LimitDegrees=20, twistLimitDegrees=20}
 {bone1Name=dyn_skirt_fr_c_03, bone2Name=dyn_skirt_fr_c_02, swing1LimitDegrees=10, swing2LimitDegrees=20, twistLimitDegrees=20}
 {bone1Name=dyn_skirt_fr_l_01, bone2Name=base_skirt_fr_l, swing1LimitDegrees=10, swing2LimitDegrees=10, twistLimitDegrees=0}
 {bone1Name=dyn_skirt_fr_l_02, bone2Name=dyn_skirt_fr_l_01, swing1LimitDegrees=10, swing2LimitDegrees=20, twistLimitDegrees=20}
 {bone1Name=dyn_skirt_fr_l_03, bone2Name=dyn_skirt_fr_l_02, swing1LimitDegrees=10, swing2LimitDegrees=20, twistLimitDegrees=20}
 {bone1Name=dyn_skirt_fr_r_01, bone2Name=base_skirt_fr_r, swing1LimitDegrees=10, swing2LimitDegrees=10, twistLimitDegrees=0}
 {bone1Name=dyn_skirt_fr_r_02, bone2Name=dyn_skirt_fr_r_01, swing1LimitDegrees=10, swing2LimitDegrees=20, twistLimitDegrees=20}
 {bone1Name=dyn_skirt_fr_r_03, bone2Name=dyn_skirt_fr_r_02, swing1LimitDegrees=10, swing2LimitDegrees=20, twistLimitDegrees=20}
 {bone1Name=dyn_skirt_sd_l_01, bone2Name=base_skirt_sd_l, swing1LimitDegrees=10, swing2LimitDegrees=10, twistLimitDegrees=0}
 {bone1Name=dyn_skirt_sd_l_02, bone2Name=dyn_skirt_sd_l_01, swing1LimitDegrees=10, swing2LimitDegrees=20, twistLimitDegrees=20}
 {bone1Name=dyn_skirt_sd_l_03, bone2Name=dyn_skirt_sd_l_02, swing1LimitDegrees=10, swing2LimitDegrees=20, twistLimitDegrees=20}
 {bone1Name=dyn_skirt_sd_r_01, bone2Name=base_skirt_sd_r, swing1LimitDegrees=10, swing2LimitDegrees=10, twistLimitDegrees=0}
 {bone1Name=dyn_skirt_sd_r_02, bone2Name=dyn_skirt_sd_r_01, swing1LimitDegrees=10, swing2LimitDegrees=20, twistLimitDegrees=20}
 {bone1Name=dyn_skirt_sd_r_03, bone2Name=dyn_skirt_sd_r_02, swing1LimitDegrees=10, swing2LimitDegrees=20, twistLimitDegrees=20}
 {
  bone1Name=base_skirt_sd_r
  bone2Name=pelvis
  swing1LimitDegrees=45
  swing1Motion=Free
  swing2LimitDegrees=45
  swing2Motion=Free
  twistLimitDegrees=45
  twistMotion=Free
 }
 {
  bone1Name=base_skirt_sd_l
  bone2Name=pelvis
  swing1LimitDegrees=45
  swing1Motion=Free
  swing2LimitDegrees=45
  swing2Motion=Free
  twistLimitDegrees=45
  twistMotion=Free
 }
 {
  bone1Name=base_skirt_fr_r
  bone2Name=pelvis
  swing1LimitDegrees=45
  swing1Motion=Free
  swing2LimitDegrees=45
  swing2Motion=Free
  twistLimitDegrees=45
  twistMotion=Free
 }
 {
  bone1Name=base_skirt_fr_c
  bone2Name=pelvis
  swing1LimitDegrees=45
  swing1Motion=Free
  swing2LimitDegrees=45
  swing2Motion=Free
  twistLimitDegrees=45
  twistMotion=Free
 }
 {
  bone1Name=base_skirt_fr_l
  bone2Name=pelvis
  swing1LimitDegrees=45
  swing1Motion=Free
  swing2LimitDegrees=45
  swing2Motion=Free
  twistLimitDegrees=45
  twistMotion=Free
 }
 {
  bone1Name=base_skirt_bk_l
  bone2Name=pelvis
  swing1LimitDegrees=45
  swing1Motion=Free
  swing2LimitDegrees=45
  swing2Motion=Free
  twistLimitDegrees=45
  twistMotion=Free
 }
 {
  bone1Name=base_skirt_bk_c
  bone2Name=pelvis
  swing1LimitDegrees=45
  swing1Motion=Free
  swing2LimitDegrees=45
  swing2Motion=Free
  twistLimitDegrees=45
  twistMotion=Free
 }
 {
  bone1Name=base_skirt_bk_r
  bone2Name=pelvis
  swing1LimitDegrees=45
  swing1Motion=Free
  swing2LimitDegrees=45
  swing2Motion=Free
  twistLimitDegrees=45
  twistMotion=Free
 }
 {
  bone1Name=thigh_l
  bone2Name=pelvis
  swing1LimitDegrees=45
  swing1Motion=Free
  swing2LimitDegrees=45
  swing2Motion=Free
  twistLimitDegrees=45
  twistMotion=Free
 }
 {
  bone1Name=thigh_r
  bone2Name=pelvis
  swing1LimitDegrees=45
  swing1Motion=Free
  swing2LimitDegrees=45
  swing2Motion=Free
  twistLimitDegrees=45
  twistMotion=Free
 }
]
